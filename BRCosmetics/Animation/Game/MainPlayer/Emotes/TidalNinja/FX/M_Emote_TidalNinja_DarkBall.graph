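# Translucent unlit Niagara sprite and mesh particle material for the dark ball effect in the TidalNinja emote.
# Use_OutSide blends an inner layer of noise-distorted, panning storm swirl with an outer plasma rim, both Fresnel-masked.
# Emissive combines panning plasma detail with the swirl, or uses flat color outside, tinted by particle color and passed through MF_FX_EmissiveAdjustments.
# Opacity follows particle alpha, softened by view-angle and depth fades.
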Material /BRCosmetics/Animation/Game/MainPlayer/Emotes/TidalNinja/FX/M_Emote_TidalNinja_DarkBall
graph {
 properties {
  bAutomaticallySetUsageInEditor=false
  bUsedWithNiagaraMeshParticles=true
  bUsedWithNiagaraSprites=true
  blendMode=BLEND_Translucent
  shadingModel=MSM_Unlit
  usedShadingModels=Unlit
 }
 nodes [
  n1 = ParticleColor()
  n2 = Multiply(
   A=TextureCoordinate() {uTiling=0.5, vTiling=0.5}
   B=AppendVector(A=ScalarParameter() {defaultValue=0.8, parameterName=BaseTextureU}, B=ScalarParameter() {defaultValue=1, parameterName=BaseTextureV})
  )
  n3 = TextureCoordinate()
  n4 = Multiply(
   A=TextureSample(UVs=Panner(Coordinate=@n3) {speedX=-0.5, speedY=0.25}) {texture=/Game/Global/Textures/Default/Flowmaps/T_SwirlFlow_02_D}.R
   B=ScalarParameter() {defaultValue=0.12, parameterName=Noise_Amount}
  )
  n5 = Constant() {r=-0.4}
  n6 = Constant() {r=0.2}
  n7 = TextureSampleParameter2D(UVs=Panner(Coordinate=Add(A=@n2, B=@n4), Speed=AppendVector(A=@n5, B=@n6)) {speedX=-0.4, speedY=0.2}) {parameterName=BaseTexture, samplerType=SAMPLERTYPE_LinearGrayscale, texture=/Game/Global/Textures/Default/Flowmaps/T_StormSwirlMask}
  n8 = Power(Base=@n7.R, Exp=ScalarParameter() {defaultValue=1, parameterName=BaseTexture_Power})
  n9 = ScalarParameter() {defaultValue=0.5, parameterName=Inside_Fresnel_Exp}
  n10 = Constant()
  n11 = Multiply(A=Multiply(A=Multiply(A=@n1.A, B=@n8), B=Fresnel(ExponentIn=@n9, BaseReflectFractionIn=@n10))) {constB=4}
  n12 = Multiply(A=TextureCoordinate() {uTiling=0.5, vTiling=0.5}, B=AppendVector(A=Constant() {r=1}, B=Constant() {r=1}))
  n13 = Constant() {r=1.6}
  n14 = Constant()
  n15 = Multiply(
   A=TextureSample(UVs=Panner(Coordinate=@n12) {speedX=-0.5, speedY=0.3}) {samplerType=SAMPLERTYPE_LinearGrayscale, texture=/Game/Effects/Fort_Effects/Emotes/UltraLight/Texture/T_Plasma_D_Greyscale}.R
   B=Power(Base=Fresnel(ExponentIn=@n13, BaseReflectFractionIn=@n14))
  )
  n16 = Multiply(A=@n1.A, B=Saturate(None=Multiply(A=@n15) {constB=10}))
  n17 = ScalarParameter() {parameterName=Use_OutSide}
  n18 = ScalarParameter() {defaultValue=1, parameterName=MeshFade}
  DepthFade(
   Opacity=Multiply(
    A=LinearInterpolate(A=@n11, B=@n16, Alpha=@n17)
    B=MaterialFunctionCall("FadeSensitivity (optional)"=@n18) {materialFunction=&FadeBasedOnViewAngle}.Result
   )
   FadeDistance=ScalarParameter() {defaultValue=50, parameterName=DepthFade}
  ) {fadeDistanceDefault=50}
  n19 = Multiply(
   A=TextureCoordinate() {uTiling=0.5, vTiling=0.5}
   B=AppendVector(A=ScalarParameter() {defaultValue=1, parameterName=DetalTextureU}, B=ScalarParameter() {defaultValue=1.5, parameterName=DetalTextureV})
  )
  n20 = AppendVector(A=Constant() {r=-0.2}, B=Constant() {r=0.12})
  n21 = Subtract(A=TextureSample(UVs=Panner(Coordinate=@n19, Speed=@n20) {speedX=-0.4, speedY=0.2}) {samplerType=SAMPLERTYPE_LinearColor, texture=/Game/Effects/Fort_Effects/Emotes/UltraLight/Texture/T_Plasma_A_C}.RGB) {constB=0.3}
  n22 = Add(A=Saturate(None=Multiply(A=@n21) {constB=2}), B=Subtract(A=@n8) {constB=0.2})
  MaterialFunctionCall(Emissive=Multiply(A=LinearInterpolate(A=@n22, B=@n1.RGB, Alpha=@n17), B=@n1.RGB)) {materialFunction=&MF_FX_EmissiveAdjustments}
 ]
}
