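# Niagara system for the purple ball explosion in the TidalNinja emote, built from five burst emitters.
# Ball and Ballsmall throw blobs slowed by drag and curl noise, layering dark, light, wave and ring attributes across sprite renderers, plus meshes on Ball.
# Flare and Glow add a central flash with a particle light, while Particle scatters flickering, speed-stretched sparks.
# A front-end preview flag, front-end modifiers and time-of-day scalars adjust color, size and brightness.
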NiagaraSystem /BRCosmetics/Animation/Game/MainPlayer/Emotes/TidalNinja/FX/NS_Emote_TidalNinja_Ball_Explosion
dependencies {
 usedDataInterfaces [&NiagaraDataInterfaceCurve, &NiagaraDataInterfaceColorCurve]
 usedDynamicInputs [&Multiply_Vector2D_ByFloat, &MakeVector]
 usedModules [
  &self:SystemSpawnScript.SystemScriptSource.SystemScriptGraph.NiagaraNodeAssignment_0.SetVariables_#1
  &SystemState
 ]
}
emitters {
 Ball {
  emitterUpdateScript [
   &EmitterState(Inactive Response=::NewEnumerator1, Life Cycle Mode=::NewEnumerator1, Loop Behavior=::NewEnumerator1, Loop Duration=1.5, UseLoopDelay=true)
   &SpawnBurst_Instantaneous(Spawn Count=5, Use Spawn Probability=false, UseLoopCountLimit=false)
  ]
  particleSpawnScript [
   &InitializeParticle(
    Color=<rgba 0.2143 0 1 1.5>
    Lifetime Max=1
    Lifetime Min=0.4
    Lifetime Mode=::NewEnumerator1
    Mesh Scale Mode=::NewEnumerator3
    Sprite Size Mode=::NewEnumerator1
    Uniform Sprite Size Max=110
    Uniform Sprite Size Min=50
    UsePositionOffset=true
    Write Material Random=true
   )
   /SetVariables(Particles.Wave_SpriteRotation:NiagaraFloat=&RandomRangeFloat(Minimum=0, Maximum=360)) {isSetParameters=true}
   &ShapeLocation(MaskSpawnGroup=false, Override Seed=false, Sphere Radius=5, UseOwnerScale=true)
   &AddVelocity(
    Module.Use Falloff Distance=false
    Override Seed=false
    Velocity Mode=::NewEnumerator1
    Velocity Speed=&RandomRangeFloat(Minimum=150, Maximum=300)
   )
   &AddVelocity(Velocity=<100 0 100>) {module=AddVelocity001}
  ]
  particleUpdateScript [
   &ParticleState()
   &ScaleSpriteSize(Uniform Curve Sprite Scale=DataInterface() {propertyValues {Curve {keys [
        <at 0 0 RCIM_Cubic>
        {arriveTangent=4.188, interpMode=RCIM_Cubic, leaveTangent=4.188, tangentMode=RCTM_User, time=0.2, value=1.625}
        {arriveTangent=-2.6562, interpMode=RCIM_Cubic, leaveTangent=-2.6562, tangentMode=RCTM_User, time=0.8, value=1.625}
        <at 1 0 RCIM_Cubic>
       ]}}}) {module=ScaleSpriteSize001}
   &ScaleColor(
    Scale Alpha=&FloatFromCurve(FloatCurve=DataInterface() {propertyValues {Curve {keys [{arriveTangent=2.1667, time=0, value=1.5}, <at 1 1.5>]}}})
    Scale RGB=&MakeVectorFromLinearColorRGB(Color=<rgba 1.5 1.5 1.5 1>)
    ScaleA=true
    ScaleRGB=true
   )
   /SetVariables(Particles.Dark_Color:LinearColor=&MakeLinearColorFromVectorAndFloat(
     "Vector (RGB)"=&MakeVectorFromLinearColorRGB(Color=&ColorFromCurve(ColorCurve=DataInterface() {propertyValues {
         AlphaCurve {keys []}
         BlueCurve {keys [<at 0.4617 0.1 RCIM_Cubic>,<at 0.5195 0.02 RCIM_Cubic>]}
         GreenCurve {keys [<at 0.4617 0.005 RCIM_Cubic>,<at 0.5195 0.001 RCIM_Cubic>]}
         RedCurve {keys [<at 0.4617 0.0254 RCIM_Cubic>,<at 0.5195 0.0051 RCIM_Cubic>]}
        }}))
     "Float (Alpha)"=&FloatFromCurve(FloatCurve=DataInterface() {propertyValues {Curve {keys [<at 0.4524 0.9974>,<at 0.5242 0.92>]}}})
    )) {isSetParameters=true, module=SetVariables_#2}
   /SetVariables(Particles.Dark_Size:Vector2f=&Multiply_Vector2D(A=$Particles.SpriteSize, B=&Vector2DFromFloat(Value=0.25))) {isSetParameters=true, module=SetVariables_#3}
   /SetVariables(Particles.Wave_Color:LinearColor=&MakeLinearColorFromVectorAndFloat(
     "Vector (RGB)"=&MakeVectorFromLinearColorRGB(Color=&MakeCustomLinearColorFromBool(Bool=$User.bIsFrontEndPreview, True Color=<rgba 2.4375 0.3 6 1>, False Color=<rgba 1.9063 0.05 5 1>))
     "Float (Alpha)"=&FloatFromCurve(FloatCurve=DataInterface() {propertyValues {Curve {keys [<at 0 0>,<at 0.1 1>,<at 0.7 1>,<at 1 0.3>]}}})
    )) {isSetParameters=true, module=SetVariables_#4}
   /SetVariables(Particles.Wave_Size:Vector2f=&Multiply_Vector2D(A=$Particles.SpriteSize, B=&Vector2DFromFloat(Value=0.22))) {isSetParameters=true, module=SetVariables_#5}
   /SetVariables(Particles.Ring_Size:Vector2f=&Vector2DFromFloat(Value=&Multiply_Float(
      A=25
      B=&FloatFromCurve(FloatCurve=DataInterface() {propertyValues {Curve {keys [<at 0 0>,<at 0.5751 0>,<at 0.6 1.75>,<at 0.7 2.25>,<at 0.9 2.25>,<at 1 3.625>]}}})
     ))) {isSetParameters=true}
   /SetVariables(Particles.Ring_Color:LinearColor=&MakeLinearColorFromVectorAndFloat(
     "Vector (RGB)"=&MakeVectorFromLinearColorRGB(Color=&ColorFromCurve(ColorCurve=DataInterface() {propertyValues {
         AlphaCurve {keys []}
         BlueCurve {keys [<at 0.4457 0.6 RCIM_Cubic>,<at 0.5514 0.6 RCIM_Cubic>]}
         GreenCurve {keys [<at 0.4457 0.108 RCIM_Cubic>,<at 0.5514 0.108 RCIM_Cubic>]}
         RedCurve {keys [<at 0.4457 0.2935 RCIM_Cubic>,<at 0.5514 0.2963 RCIM_Cubic>]}
        }}))
     "Float (Alpha)"=&FloatFromCurve(FloatCurve=DataInterface() {propertyValues {Curve {keys [<at 0 0.4>,<at 0.5 0.4>,<at 0.6 0.9>,<at 0.8 1>]}}})
    )) {isSetParameters=true}
   /SetVariables(Particles.Wave_Detail_Color:LinearColor=&MakeLinearColorFromVectorAndFloat(
     "Vector (RGB)"=&MakeVectorFromLinearColorRGB(Color=<rgba 0.7625 0.02 2 1>)
     "Float (Alpha)"=&FloatFromCurve(
      FloatCurve=DataInterface() {propertyValues {Curve {keys [
          <at 0 0 RCIM_Cubic>
          {arriveTangent=2.1745, interpMode=RCIM_Cubic, leaveTangent=2.1745, tangentMode=RCTM_User, time=0.2, value=0.75}
          {arriveTangent=-1.7254, interpMode=RCIM_Cubic, leaveTangent=-1.7254, tangentMode=RCTM_User, time=0.6, value=0.8}
          <at 1 0 RCIM_Cubic>
         ]}}}
      Scale Curve=&RandomRangeFloat(Minimum=0.5, Maximum=1)
     )
    )) {enabled=false, isSetParameters=true}
   /SetVariables(Particles.Wave_Detail_Scale:Vector3f=&VectorFromFloat(Value=&FloatFromCurve(
      FloatCurve=DataInterface() {propertyValues {Curve {keys [
          <at 0 0 RCIM_Cubic>
          {arriveTangent=5.7596, interpMode=RCIM_Cubic, leaveTangent=5.7596, tangentMode=RCTM_User, time=0.2, value=1}
          {arriveTangent=-3.3002, interpMode=RCIM_Cubic, leaveTangent=-3.3002, tangentMode=RCTM_User, time=0.7731, value=1.3102}
          <at 1 0 RCIM_Cubic>
         ]}}}
      Scale Curve=0.1
     ))) {enabled=false, isSetParameters=true}
   /SetVariables(Particles.Wave_Detail_Dyn:Vector4f=&MakeVector4(
     X=&FloatFromCurve(FloatCurve=DataInterface() {propertyValues {Curve {keys [<at 0.5 0>,<at 1 1>]}}})
     Y=0
     Z=0
     W=0
    )) {isSetParameters=true}
   /SetVariables(Particles.Wave_Detail_Center_Scale:Vector3f=&VectorFromFloat(Value=&FloatFromCurve(
      FloatCurve=DataInterface() {propertyValues {Curve {keys [
          <at 0 0 RCIM_Cubic>
          {arriveTangent=3.2846, interpMode=RCIM_Cubic, leaveTangent=3.2846, tangentMode=RCTM_User, time=0.2, value=1.3}
          {arriveTangent=-1.75, interpMode=RCIM_Cubic, leaveTangent=-1.75, time=0.8419, value=1.4989}
          <at 1 -0.1 RCIM_Cubic>
         ]}}}
      Scale Curve=&RandomRangeFloat(Minimum=0.04, Maximum=0.08)
     ))) {isSetParameters=true}
   /SetVariables(Particles.Wave_Detail_Center_Color:LinearColor=&MakeLinearColorFromVectorAndFloat(
     "Vector (RGB)"=&MakeVectorFromLinearColorRGB(Color=<rgba 0.08 0.0021 0.2 1>)
     "Float (Alpha)"=&RandomRangeFloat(Minimum=0.5, Maximum=1)
    )) {isSetParameters=true}
   /SetVariables(Particles.Light-Color:LinearColor=&MakeLinearColorFromVectorAndFloat(
     "Vector (RGB)"=&MakeVectorFromLinearColorRGB(Color=&ColorFromCurve(ColorCurve=DataInterface() {propertyValues {
         AlphaCurve {keys []}
         BlueCurve {keys [<at 0.4617 1 RCIM_Cubic>,<at 0.9154 0.02 RCIM_Cubic>]}
         GreenCurve {keys [<at 0.4617 0 RCIM_Cubic>,<at 0.9154 0.001 RCIM_Cubic>]}
         RedCurve {keys [<at 0.4617 0.3333 RCIM_Cubic>,<at 0.9154 0.0051 RCIM_Cubic>]}
        }}))
     "Float (Alpha)"=&FloatFromCurve(
      FloatCurve=DataInterface() {propertyValues {Curve {keys [<at 0 1>,<at 0.4524 0.9974>,<at 0.5 0.5>,<at 1 0.5>]}}}
      Scale Curve=&MakeCustomFloatFromBool(Bool=$User.bIsFrontEndPreview, True Float=0.8, False Float=1)
     )
    )) {isSetParameters=true}
   /SetVariables(Particles.Light_Size:Vector2f=&Multiply_Vector2D(
     A=$Particles.SpriteSize
     B=&MakeCustomVector2DFromBool(Bool=$User.bIsFrontEndPreview, True Vector2D=<0.8 0.3>, False Vector2D=<1 0.4>)
    )) {isSetParameters=true}
   &CurlNoiseForce(Bias Noise=false, Noise Strength=600, Pan Noise=false)
   &Drag(Drag=4, Use Linear Drag=true, Use Rotational Drag=true)
   &SolveForcesAndVelocity(Clamp Velocity=false, Limit Acceleration=false)
  ]
  rendererClasses [&NiagaraSpriteRendererProperties, &NiagaraMeshRendererProperties]
  renderers [
   {rendererClass=&NiagaraSpriteRendererProperties}
   {rendererClass=&NiagaraSpriteRendererProperties, rendererIndex=1}
   {rendererClass=&NiagaraMeshRendererProperties, rendererIndex=2}
   {rendererClass=&NiagaraMeshRendererProperties, rendererIndex=3}
   {rendererClass=&NiagaraSpriteRendererProperties, rendererIndex=4}
  ]
 }
 Ballsmall {
  emitterUpdateScript [
   &EmitterState(Inactive Response=::NewEnumerator1, Life Cycle Mode=::NewEnumerator1, Loop Behavior=::NewEnumerator1, Loop Duration=1.5, UseLoopDelay=true)
   &SpawnBurst_Instantaneous(Spawn Count=4, Use Spawn Probability=false, UseLoopCountLimit=false)
  ]
  particleSpawnScript [
   &InitializeParticle(
    Color=<rgba 0.4113 0.0524 1 2>
    Lifetime Max=1
    Lifetime Min=0.4
    Lifetime Mode=::NewEnumerator1
    Sprite Size Mode=::NewEnumerator1
    Uniform Sprite Size Max=150
    Uniform Sprite Size Min=90
    UsePositionOffset=true
    Write Material Random=true
   )
   &ShapeLocation(MaskSpawnGroup=false, Override Seed=false, Sphere Radius=5, UseOwnerScale=true)
   &AddVelocity(
    Module.Use Falloff Distance=false
    Override Seed=false
    Velocity Mode=::NewEnumerator1
    Velocity Speed=&Multiply_Float(
     A=1000
     B=&FloatFromCurve(FloatCurve=DataInterface() {propertyValues {Curve {keys [<at 0 0>,<at 0.1 2.375>,<at 0.4 0.75>,<at 1 0>]}}}, CurveIndex=$Emitter.Age)
    )
   )
   &AddVelocity(Velocity=<50 0 50>) {module=AddVelocity001}
  ]
  particleUpdateScript [
   &ParticleState()
   &ScaleSpriteSize(Uniform Curve Sprite Scale=DataInterface() {propertyValues {Curve {keys [
        {arriveTangent=0.1979, interpMode=RCIM_Cubic, leaveTangent=0.1979, tangentMode=RCTM_User, time=0, value=0}
        {arriveTangent=1.5051, interpMode=RCIM_Cubic, leaveTangent=1.5051, tangentMode=RCTM_User, time=0.1, value=0.75}
        {arriveTangent=-0.3352, interpMode=RCIM_Cubic, leaveTangent=-0.3352, tangentMode=RCTM_User, time=0.7, value=1}
        {arriveTangent=0.1979, interpMode=RCIM_Cubic, leaveTangent=0.1979, tangentMode=RCTM_User, time=1, value=0}
       ]}}}) {module=ScaleSpriteSize001}
   &ScaleColor(Scale RGB=&MakeVectorFromLinearColorRGB(Color=<rgba 0.5511 0.1595 1 1>), ScaleA=true, ScaleRGB=true)
   /SetVariables(Particles.Dark_Color:LinearColor=&MakeLinearColorFromVectorAndFloat("Vector (RGB)"=&MakeVectorFromLinearColorRGB(Color=<rgba 0.0214 0.003 0.06 1>), "Float (Alpha)"=3)) {isSetParameters=true, module=SetVariables_#2}
   /SetVariables(Particles.Dark_Size:Vector2f=&Multiply_Vector2D(A=$Particles.SpriteSize, B=&Vector2DFromFloat(Value=0.3))) {isSetParameters=true, module=SetVariables_#3}
   /SetVariables(Particles.Dark_Dyn:Vector4f=&MakeVector4(X=5, Y=1, Z=0, W=0)) {isSetParameters=true}
   /SetVariables(Particles.Light_Size:Vector2f=&Multiply_Vector2D(A=$Particles.SpriteSize, B=&RandomRangeVector2D(Minimum=<0.4 0.2>, Maximum=<0.6 0.25>))) {isSetParameters=true}
   &CurlNoiseForce(Bias Noise=false, Noise Strength=800, Pan Noise=false)
   &Drag(
    Drag=&FloatFromCurve(FloatCurve=DataInterface() {propertyValues {Curve {keys [<at 0 1>,<at 0.5 1>,<at 1 0>]}}}, Scale Curve=10)
    Use Linear Drag=true
    Use Rotational Drag=true
   )
   &SolveForcesAndVelocity(Clamp Velocity=false, Limit Acceleration=false)
  ]
  rendererClasses [&NiagaraSpriteRendererProperties]
  renderers [
   {rendererClass=&NiagaraSpriteRendererProperties}
   {rendererClass=&NiagaraSpriteRendererProperties, rendererIndex=1}
  ]
 }
 Flare {
  emitterUpdateScript [
   &EmitterState(Inactive Response=::NewEnumerator1, Life Cycle Mode=::NewEnumerator1, Loop Behavior=::NewEnumerator1, Loop Duration=1, UseLoopDelay=true)
   &SpawnBurst_Instantaneous(Use Spawn Probability=false, UseLoopCountLimit=false)
  ]
  particleSpawnScript [
   &InitializeParticle(
    Color=<rgba 0.2143 0 1 2>
    Lifetime=0.52
    Sprite Size Mode=::NewEnumerator3
    Uniform Sprite Size=120
    UsePositionOffset=false
    Write Material Random=true
   )
   /SetVariables(Particles.Wave_SpriteRotation:NiagaraFloat=&RandomRangeFloat(Minimum=0, Maximum=360)) {isSetParameters=true}
  ]
  particleUpdateScript [
   &ParticleState()
   &ScaleSpriteSize(Uniform Curve Sprite Scale=DataInterface() {propertyValues {Curve {keys [
        {arriveTangent=4.4111, interpMode=RCIM_Cubic, leaveTangent=4.4111, tangentMode=RCTM_User, time=0, value=0}
        {arriveTangent=1, interpMode=RCIM_Cubic, leaveTangent=1, time=0.2, value=0.65}
        <at 1 1 RCIM_Cubic>
       ]}}}) {module=ScaleSpriteSize001}
   &ScaleColor(
    Scale Alpha=&FloatFromCurve(FloatCurve=DataInterface() {propertyValues {Curve {keys [{arriveTangent=2.1667, time=0, value=1.5}, <at 1 1.5>]}}})
    Scale RGB=&MakeVectorFromLinearColorRGB(Color=<rgba 1.5 1.5 1.5 1>)
    ScaleA=true
    ScaleRGB=true
   )
   &ShapeLocation(MaskSpawnGroup=false, Offset=$System.Location, Offset Mode=::NewEnumerator0, Override Seed=false, Sphere Radius=0.01, UseOwnerScale=true)
   /SetVariables(Particles.Wave_Color:LinearColor=&MakeLinearColorFromVectorAndFloat(
     "Vector (RGB)"=&MakeVectorFromLinearColorRGB(Color=&MakeCustomLinearColorFromBool(Bool=$User.bIsFrontEndPreview, True Color=<rgba 7.625 0.2 20 1>, False Color=<rgba 15.25 0.4 40 1>))
     "Float (Alpha)"=&FloatFromCurve(FloatCurve=DataInterface() {propertyValues {Curve {keys [{arriveTangent=-2.2, interpMode=RCIM_Cubic, leaveTangent=-2.2, tangentMode=RCTM_User, time=0, value=1}, <at 1 0 RCIM_Cubic>]}}})
    )) {isSetParameters=true, module=SetVariables_#4}
   /SetVariables(Particles.Wave_Size:Vector2f=&Multiply_Vector2D(
     A=$Particles.SpriteSize
     B=&Vector2DFromFloat(Value=&FloatFromCurve(
       FloatCurve=DataInterface() {propertyValues {Curve {keys [
           {arriveTangent=1.2508, interpMode=RCIM_Cubic, leaveTangent=1.2508, tangentMode=RCTM_User, time=0, value=0.5}
           {arriveTangent=0.5741, interpMode=RCIM_Cubic, leaveTangent=0.5741, tangentMode=RCTM_User, time=0.2, value=0.75}
           <at 1 1 RCIM_Cubic>
          ]}}}
       Scale Curve=1.2
      ))
    )) {isSetParameters=true, module=SetVariables_#5}
  ]
  rendererClasses [&NiagaraSpriteRendererProperties]
  renderers [{rendererClass=&NiagaraSpriteRendererProperties}]
 }
 Glow {
  emitterUpdateScript [
   &EmitterState(
    Inactive Response=::NewEnumerator1
    Life Cycle Mode=::NewEnumerator1
    Loop Behavior=::NewEnumerator1
    Loop Delay=1.6
    Loop Duration=0.5
    UseLoopDelay=false
   )
   &SpawnBurst_Instantaneous(Use Spawn Probability=false, UseLoopCountLimit=false)
  ]
  particleSpawnScript [
   &InitializeParticle(
    Color=<rgba 0.809 0.1924 3 1>
    Lifetime=0.3
    Sprite Size Mode=::NewEnumerator3
    Uniform Sprite Size=45
    UsePositionOffset=false
    Write Material Random=true
   )
   &CameraOffset(Camera Offset Amount=80)
  ]
  particleUpdateScript [
   &ParticleState()
   &ScaleSpriteSize(
    Uniform Curve Scale=&FN_FrontEndFloatModifier(Gameplay=1, Front End=0.8)
    Uniform Curve Sprite Scale=DataInterface() {propertyValues {Curve {keys [
        {arriveTangent=5.1246, interpMode=RCIM_Cubic, leaveTangent=5.1246, tangentMode=RCTM_User, time=0, value=0}
        {arriveTangent=1.7949, interpMode=RCIM_Cubic, leaveTangent=1.7949, tangentMode=RCTM_User, time=0.3, value=1.5}
        <at 1 2 RCIM_Cubic>
       ]}}}
   )
   &ScaleColor(
    Scale Alpha=&FloatFromCurve(
     FloatCurve=DataInterface() {propertyValues {Curve {keys [
         {arriveTangent=-2.7027, interpMode=RCIM_Cubic, leaveTangent=-2.7027, tangentMode=RCTM_User, time=0, value=1}
         {arriveTangent=-1, interpMode=RCIM_Cubic, leaveTangent=-1, time=0.312, value=0.3309}
         <at 1 0>
        ]}}}
     Scale Curve=&Multiply_Float(A=1, B=&FNTimeOfDayScalar(Night Scale=0.8, Morning Scale=1, Evening Scale=1))
    )
    ScaleA=true
    ScaleRGB=true
   )
   &ShapeLocation(MaskSpawnGroup=false, Offset=$System.Location, Offset Mode=::NewEnumerator0, Override Seed=false, Sphere Radius=0.01, UseOwnerScale=true)
   &FN_Light_Attributes(
    Color=&ColorFromCurve(ColorCurve=DataInterface() {propertyValues {
       AlphaCurve {keys []}
       BlueCurve {keys [<at 0.2351 4 RCIM_Cubic>,<at 0.9742 2 RCIM_Cubic>]}
       GreenCurve {keys [<at 0.2351 0.7033 RCIM_Cubic>,<at 0.9742 0.1262 RCIM_Cubic>]}
       RedCurve {keys [<at 0.2351 1.2129 RCIM_Cubic>,<at 0.9742 0.1711 RCIM_Cubic>]}
      }})
    Create Color Attribute=true
    Create Exponent Attribute=true
    Create Position Attribute=true
    Create Radius Attribute=true
    Enable Particle Light Enable=true
    Exponent=2
    Position=<0 0 50>
    Radius=&Multiply_Float(
     A=100
     B=&FloatFromCurve(FloatCurve=DataInterface() {propertyValues {Curve {keys [<at 0 1.8 RCIM_Cubic>,<at 0.2 2 RCIM_Cubic>,<at 1 1.8 RCIM_Cubic>]}}})
    )
    Scale Color=&FloatFromCurve(
     FloatCurve=DataInterface() {propertyValues {Curve {keys [
         {arriveTangent=5.626, interpMode=RCIM_Cubic, leaveTangent=5.626, tangentMode=RCTM_User, time=0, value=0.6}
         {arriveTangent=0.0895, interpMode=RCIM_Cubic, leaveTangent=0.0895, tangentMode=RCTM_User, time=0.1, value=1}
         {arriveTangent=-0.8586, interpMode=RCIM_Cubic, leaveTangent=-0.8586, tangentMode=RCTM_User, time=0.815, value=0.72}
         {arriveTangent=-4.4162, interpMode=RCIM_Cubic, leaveTangent=-4.4162, tangentMode=RCTM_User, time=1, value=0}
        ]}}}
     Scale Curve=&Multiply_Float(
      A=&FN_FrontEndFloatModifier(Gameplay=30, Front End=5)
      B=&FNTimeOfDayScalar(Night Scale=0.5, Morning Scale=1, Evening Scale=0.8)
     )
    )
   )
   /SetVariables(Particles.Dark_Color:LinearColor=&MakeLinearColorFromVectorAndFloat(
     "Vector (RGB)"=&MakeVectorFromLinearColorRGB(Color=<rgba 0.0043 0 0.02 1>)
     "Float (Alpha)"=&FloatFromCurve(
      FloatCurve=DataInterface() {propertyValues {Curve {keys [{arriveTangent=-1.4568, interpMode=RCIM_Cubic, leaveTangent=-1.4568, tangentMode=RCTM_User, time=0, value=1}, <at 1 0 RCIM_Cubic>]}}}
      Scale Curve=3
     )
    )) {isSetParameters=true}
   /SetVariables(Particles.Dark_Size:Vector2f=&Multiply_Vector2D(A=$Particles.SpriteSize, B=&Vector2DFromFloat(Value=1.5))) {isSetParameters=true}
   /SetVariables(Particles.Dark_Dyn:Vector4f=&MakeVector4(X=2, Y=1, Z=0, W=0)) {isSetParameters=true}
   /SetVariables(Particles.Light_Size:Vector2f=&Multiply_Vector2D(
     A=<280 210>
     B=&Vector2DFromFloat(Value=&FloatFromCurve(FloatCurve=DataInterface() {propertyValues {Curve {keys [<at 0 1>,<at 0.3 0.3>,<at 0.5 0.9>,<at 1 0>]}}}))
    )) {isSetParameters=true}
   /SetVariables(Particles.Light-Color:LinearColor=&MakeLinearColorFromVectorAndFloat(
     "Vector (RGB)"=&MakeVectorFromLinearColorRGB(Color=&ColorFromCurve(ColorCurve=DataInterface() {propertyValues {
         AlphaCurve {keys []}
         BlueCurve {keys [<at 0.0805 200>, {arriveTangent=-157.524, leaveTangent=-157.524, time=0.3203, value=5}, <at 0.517 1>]}
         GreenCurve {keys [<at 0.0805 12.8244>, {arriveTangent=-10.1007, leaveTangent=-10.1007, time=0.3203, value=0.3206}, <at 0.517 0.0641>]}
         RedCurve {keys [<at 0.0805 53.9366>, {arriveTangent=-42.4815, leaveTangent=-42.4815, time=0.3203, value=1.3484}, <at 0.517 0.2697>]}
        }}))
     "Float (Alpha)"=1
    )) {isSetParameters=true}
   &SolveForcesAndVelocity(Clamp Velocity=false, Limit Acceleration=false)
  ]
  rendererClasses [&NiagaraSpriteRendererProperties, &NiagaraLightRendererProperties]
  renderers [
   {rendererClass=&NiagaraSpriteRendererProperties}
   {rendererClass=&NiagaraLightRendererProperties, rendererIndex=1}
   {rendererClass=&NiagaraSpriteRendererProperties, rendererIndex=2}
  ]
 }
 Particle {
  emitterUpdateScript [
   &EmitterState(Inactive Response=::NewEnumerator1, Life Cycle Mode=::NewEnumerator1, Loop Behavior=::NewEnumerator1, Loop Duration=1, UseLoopDelay=false)
   &SpawnBurst_Instantaneous(Spawn Count=20, Use Spawn Probability=false, UseLoopCountLimit=false)
  ]
  particleSpawnScript [
   &InitializeParticle(
    Color=<rgba 1.8703 0 10 1>
    Lifetime Max=0.8
    Lifetime Min=0.1
    Lifetime Mode=::NewEnumerator1
    Sprite Size Max=<2 4>
    Sprite Size Min=<1.5 3>
    Sprite Size Mode=::NewEnumerator2
    UsePositionOffset=true
    Write Material Random=true
   )
   &ShapeLocation(
    MaskSpawnGroup=false
    Offset=<0 0 20>
    Offset Mode=::NewEnumerator0
    Override Seed=false
    Rotation Mode=::NewEnumerator1
    Sphere Radius=20
    UseOwnerScale=true
   )
   &AddVelocity(
    Module.Use Falloff Distance=false
    Origin Offset=<0 0 25>
    Override Seed=false
    Velocity Mode=::NewEnumerator1
    Velocity Speed=&RandomRangeFloat(Minimum=50, Maximum=300)
   )
   &AddVelocity(Velocity=<100 0 100>) {module=AddVelocity001}
  ]
  particleUpdateScript [
   &ParticleState()
   &ScaleSpriteSize(Uniform Curve Sprite Scale=DataInterface() {propertyValues {Curve {keys [
        <at 0 0>
        {arriveTangent=1.5, leaveTangent=1.5, time=0.0854, value=0.85}
        <at 0.2705 0.0312>
        <at 0.3267 0.9335>
        {arriveTangent=-0.25, leaveTangent=-0.25, time=0.4998, value=0.1033}
        {arriveTangent=-0.5, leaveTangent=-0.5, time=0.6227, value=0.9345}
        <at 0.8 0.05>
        <at 0.9 0.9>
        <at 1 0>
       ]}}})
   &CurlNoiseForce(Noise Strength=1000, Pan Noise=true, Pan Noise Field=<0 0.2 0.2>)
   &Drag(Drag=&RandomRangeFloat(Minimum=2, Maximum=3), Use Linear Drag=true, Use Rotational Drag=true)
   &AccelerationForce(Acceleration=<0 0 50>)
   &SolveForcesAndVelocity(Clamp Velocity=false, Limit Acceleration=false)
   &ScaleColor(
    Scale Alpha=&FloatFromCurve(
     FloatCurve=DataInterface() {propertyValues {Curve {keys [
         {interpMode=RCIM_Cubic, tangentMode=RCTM_User, time=0, value=0}
         {interpMode=RCIM_Cubic, tangentMode=RCTM_User, time=0.5, value=1}
         {tangentMode=RCTM_User, time=1, value=0}
        ]}}}
     Scale Curve=&MakeCustomFloatFromBool(Bool=$User.bIsFrontEndPreview, True Float=0.5, False Float=1)
    )
    ScaleA=true
    ScaleRGB=true
   )
   &FN_Sprite_Attributes(
    Color=&MultiplyLinearColor(
     Linear Color=<rgba 0.2963 0.097 1 1>
     Scale Factor=&MakeVector4_FromFloat(Float=&FloatFromCurve(
       FloatCurve=DataInterface() {propertyValues {Curve {keys [
           {interpMode=RCIM_Cubic, tangentMode=RCTM_User, time=0, value=0}
           {interpMode=RCIM_Cubic, tangentMode=RCTM_User, time=0.5, value=0.8}
           {tangentMode=RCTM_User, time=1, value=0}
          ]}}}
       Scale Curve=&MakeCustomFloatFromBool(Bool=$User.bIsFrontEndPreview, True Float=0.5, False Float=1)
      ))
    )
    Create Color Attribute=true
    Create SpriteSize Attribute=true
    SpriteSize=&Multiply_Vector2D(
     A=&RandomRangeVector2D(Minimum=<3 5>, Maximum=<4 6>)
     B=&Vector2DFromFloat(Value=&FloatFromCurve(
       FloatCurve=DataInterface() {propertyValues {Curve {keys [<at 0 0>,<at 0.1 0.9>,<at 0.2 0.35>,<at 0.3 1>,<at 0.4 0.15>,<at 0.5 0.9>,<at 0.6 0.5>,<at 0.7 0.9>,<at 0.8 0.35>,<at 0.9 0.95>,<at 1 0>]}}}
       Scale Curve=&RandomRangeFloat(Minimum=3, Maximum=4)
      ))
    )
    Write to Module Namespace=true
   )
   &ScaleSpriteSizeBySpeed(Max Scale Factor=<1.5 10>)
  ]
  rendererClasses [&NiagaraSpriteRendererProperties]
  renderers [
   {rendererClass=&NiagaraSpriteRendererProperties}
   {rendererClass=&NiagaraSpriteRendererProperties, rendererIndex=1}
  ]
 }
}
summary {
 systemName=NS_Emote_TidalNinja_Ball_Explosion
 emitters [
  {emitterName=Flare, rendererClasses [&NiagaraSpriteRendererProperties]}
  {emitterName=Glow, rendererClasses [&NiagaraSpriteRendererProperties, &NiagaraLightRendererProperties]}
  {emitterName=Particle, rendererClasses [&NiagaraSpriteRendererProperties]}
  {emitterName=Ball, rendererClasses [&NiagaraSpriteRendererProperties, &NiagaraMeshRendererProperties]}
  {emitterName=Ballsmall, rendererClasses [&NiagaraSpriteRendererProperties]}
 ]
 userVariables [
  User.bIsFrontEnd {type=NiagaraBool}
  User.bIsFrontEndPreview {type=NiagaraBool}
  User.VariantSwitch {type=NiagaraFloat, defaultValue=0, description=""}
 ]
}
