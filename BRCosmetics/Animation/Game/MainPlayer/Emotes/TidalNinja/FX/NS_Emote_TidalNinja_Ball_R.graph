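# Niagara effect for the TidalNinja emote's "R" ball: a violet energy orb that grows, pulses and swells over 5.6 seconds.
# One Ball particle drives layered sprite, mesh and light renderers through curve-animated custom attributes, with wave color switched for front-end preview.
# Movement-spawned ribbon and sparks trail it, while delayed absorb, smoke-ring and lightning emitters build the charge-up.
# Center flare and smoke opacity scale by time of day.
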NiagaraSystem /BRCosmetics/Animation/Game/MainPlayer/Emotes/TidalNinja/FX/NS_Emote_TidalNinja_Ball_R
dependencies {
 usedDataInterfaces [&NiagaraDataInterfaceCurve, &NiagaraDataInterfaceColorCurve, &NiagaraDataInterfaceVector2DCurve]
 usedDynamicInputs []
 usedModules [&SystemState]
}
emitters {
 Absorb_Beam {
  emitterUpdateScript [
   &EmitterState(Inactive Response=::NewEnumerator1, Life Cycle Mode=::NewEnumerator1, Loop Behavior=::NewEnumerator1, Loop Duration=3, UseLoopDelay=true)
   &SpawnBurst_Instantaneous(Use Spawn Probability=false, UseLoopCountLimit=false) {enabled=false}
   &SpawnRate(
    SpawnRate=&FloatFromCurve(
     FloatCurve=DataInterface() {propertyValues {Curve {keys [<at 0 0>,<at 0.2972 2>,<at 0.6315 1.1773>,<at 0.6932 0>,<at 1.1706 0>,<at 1.2191 0.5674>,<at 1.412 1.341>,<at 1.6728 1.5>]}}}
     CurveIndex=$Emitter.Age
     Scale Curve=20
    )
    Use Spawn Probability=false
   )
  ]
  particleSpawnScript [
   &InitializeParticle(
    Lifetime=&Multiply_Float(
     A=0.35
     B=&FloatFromCurve(
      FloatCurve=DataInterface() {propertyValues {Curve {keys [<at 0 1>,<at 1.5 0.45>,<at 1.8 0.5>,<at 3 0.3>]}}}
      CurveIndex=$Emitter.NormalizedLoopAge
     )
    )
    Mesh Scale Max=<1 1 1>
    Mesh Scale Min=<0.8 0.8 0.8>
    Mesh Scale Mode=::NewEnumerator2
    UsePositionOffset=false
    Write Material Random=true
   )
   &InitialMeshOrientation(
    Rotation=&RandomRangeVector(Minimum=<-1 -1 -1>, Maximum=<1 1 1>)
    Use Orientation Vector=true
    Use Rotation Vector=true
   )
  ]
  particleUpdateScript [
   &ParticleState()
   &ScaleMeshSize(Scale Factor=&VectorFromFloat(Value=&Multiply_Float(
      A=&RandomRangeFloat(Minimum=0.5, Maximum=0.6)
      B=&FloatFromCurve(FloatCurve=DataInterface() {propertyValues {Curve {keys [<at 0 0.5>,<at 0.8 0.5>,<at 2 1.4>]}}}, CurveIndex=$Emitter.Age)
     )))
   &SolveForcesAndVelocity(Clamp Velocity=false, Limit Acceleration=false)
   &DynamicMaterialParameters(
    Index 0 Param 1=&FloatFromCurve(FloatCurve=DataInterface() {propertyValues {Curve {keys [<at 0 0>,<at 1 1.2>]}}})
    Index 0 Param 2=1
    Index 0 Param 3=&Multiply_Float(
     A=&RandomRangeFloat(Minimum=0.5, Maximum=1)
     B=&FloatFromCurve(
      FloatCurve=DataInterface() {propertyValues {Curve {keys [<at 0 1>,<at 1 0.5>]}}}
      Scale Curve=&FloatFromCurve(FloatCurve=DataInterface() {propertyValues {Curve {keys [<at 0 0.3>,<at 2 1>]}}}, CurveIndex=$Emitter.Age)
     )
    )
    Module.Param0WriteEnabled=true
    Module.Param1WriteEnabled=true
    Module.Param2WriteEnabled=true
    Module.Param3WriteEnabled=true
   )
   &ScaleColor(
    Scale Alpha=&FloatFromCurve(FloatCurve=DataInterface() {propertyValues {Curve {keys [<at 0 0 RCIM_Cubic>,<at 0.4 1 RCIM_Cubic>,<at 1 0 RCIM_Cubic>]}}}, Scale Curve=2)
    Scale RGB=&MakeVectorFromLinearColorRGB(Color=&ColorFromCurve(ColorCurve=DataInterface() {propertyValues {
        AlphaCurve {keys []}
        BlueCurve {keys [<at 0.1336 1.5 RCIM_Cubic>,<at 0.4118 0.09 RCIM_Cubic>]}
        GreenCurve {keys [<at 0.1336 0.1462 RCIM_Cubic>,<at 0.4118 0.001 RCIM_Cubic>]}
        RedCurve {keys [<at 0.1336 0.6072 RCIM_Cubic>,<at 0.4118 0.0343 RCIM_Cubic>]}
       }}))
    ScaleA=true
    ScaleRGB=true
   )
  ]
  rendererClasses [&NiagaraMeshRendererProperties]
  renderers [{rendererClass=&NiagaraMeshRendererProperties, rendererIndex=0}]
 }
 Absorb_Energy {
  emitterUpdateScript [
   &EmitterState(
    Inactive Response=::NewEnumerator1
    Life Cycle Mode=::NewEnumerator1
    Loop Behavior=::NewEnumerator1
    Loop Delay=2
    Loop Duration=1.5
    UseLoopDelay=true
   )
   &SpawnRate(SpawnRate=3, Use Spawn Probability=false) {enabled=false}
   &SpawnBurst_Instantaneous(Use Spawn Probability=false, UseLoopCountLimit=false)
  ]
  particleSpawnScript [
   &InitializeParticle(Lifetime=1.2, Mesh Scale Mode=::NewEnumerator3, Mesh Uniform Scale=1.5, UsePositionOffset=false, Write Material Random=true)
   &InitialMeshOrientation(
    Mesh Orientation Mode=::NewEnumerator2
    Rotation=&RandomRangeVector(Minimum=<0 0 0>, Maximum=<1 0 0>)
    Use Rotation Vector=true
   )
  ]
  particleUpdateScript [
   &ParticleState()
   &ScaleColor(
    Scale Alpha=&FloatFromCurve(FloatCurve=DataInterface() {propertyValues {Curve {keys [
         {arriveTangent=0.0004, interpMode=RCIM_Cubic, leaveTangent=0.0004, tangentMode=RCTM_User, time=0, value=0}
         {arriveTangent=0.02, interpMode=RCIM_Cubic, leaveTangent=0.02, tangentMode=RCTM_User, time=0.1, value=1}
         {arriveTangent=-0.0087, interpMode=RCIM_Cubic, leaveTangent=-0.0087, tangentMode=RCTM_User, time=0.8, value=1}
         {arriveTangent=0.0459, interpMode=RCIM_Cubic, leaveTangent=0.0459, tangentMode=RCTM_User, time=1, value=0}
        ]}}})
    Scale RGB=&MakeVectorFromLinearColorRGB(Color=<rgba 1 1 1 1>)
    ScaleA=true
    ScaleRGB=true
   )
   &DynamicMaterialParameters(
    Index 0 Param 3=1.5
    Module.Index0Float0Write=true
    Module.Index0Float1Write=true
    Module.Index0Float2Write=true
    Module.Index0Float3Write=true
   )
   &ScaleColor(
    Scale Alpha=&FloatFromCurve(
     FloatCurve=DataInterface() {propertyValues {Curve {keys [
         {arriveTangent=0.0005, interpMode=RCIM_Cubic, leaveTangent=0.0005, tangentMode=RCTM_User, time=0, value=0}
         {arriveTangent=0.0014, interpMode=RCIM_Cubic, leaveTangent=0.0014, tangentMode=RCTM_User, time=2.4, value=1}
         {arriveTangent=0.0005, interpMode=RCIM_Cubic, leaveTangent=0.0005, tangentMode=RCTM_User, time=2.8, value=1}
         {arriveTangent=-10.1543, interpMode=RCIM_Cubic, leaveTangent=-10.1543, tangentMode=RCTM_User, time=3, value=0}
        ]}}}
     CurveIndex=$Emitter.Age
    )
    ScaleA=true
    ScaleRGB=true
   ) {enabled=false, module=ScaleColor002}
   &ScaleMeshSize(Scale Factor=&VectorFromFloat(Value=&FloatFromCurve(FloatCurve=DataInterface() {propertyValues {Curve {keys [{arriveTangent=0.4936, interpMode=RCIM_Cubic, leaveTangent=0.4936, tangentMode=RCTM_User, time=0, value=0.4}, <at 1 1 RCIM_Cubic>]}}})))
  ]
  rendererClasses [&NiagaraMeshRendererProperties]
  renderers [{rendererClass=&NiagaraMeshRendererProperties, rendererIndex=0}]
 }
 Absorb_Particles {
  emitterUpdateScript [
   &EmitterState(
    Inactive Response=::NewEnumerator1
    Life Cycle Mode=::NewEnumerator1
    Loop Behavior=::NewEnumerator1
    Loop Delay=1.6
    Loop Duration=1.5
    UseLoopDelay=true
   )
   &SpawnRate(
    SpawnRate=&Multiply_Float(
     A=100
     B=&FloatFromCurve(
      FloatCurve=DataInterface() {propertyValues {Curve {keys [{interpMode=RCIM_Cubic, tangentMode=RCTM_User, time=0, value=1}, {interpMode=RCIM_Cubic, tangentMode=RCTM_User, time=1, value=2}]}}}
      CurveIndex=$Emitter.NormalizedLoopAge
     )
    )
    Use Spawn Probability=false
   )
  ]
  particleSpawnScript [
   &InitializeParticle(
    Lifetime=&Multiply_Float(
     A=0.2
     B=&FloatFromCurve(FloatCurve=DataInterface() {propertyValues {Curve {keys [<at 0 1>,<at 1 0.4>]}}})
    )
    Sprite Size Max=<1 5>
    Sprite Size Min=<1 4>
    Sprite Size Mode=::NewEnumerator2
    UsePositionOffset=false
    Write Material Random=true
   )
   &ShapeLocation(MaskSpawnGroup=false, Override Seed=false, Sphere Radius=90, UseOwnerScale=true)
   &CurlNoiseForce(Noise Frequency=10, Noise Strength=-100, Pan Noise=false) {enabled=false, module=CurlNoiseForce001}
   &ApplyInitialForces()
  ]
  particleUpdateScript [
   &ParticleState()
   &CurlNoiseForce(Bias Noise=false, Noise Strength=3000, Pan Noise=false)
   &PointForce(
    Force Origin=&AddVectorToPosition(Position=$Particles.SkeletalMeshLocation001.SampledPosition, Vector=<0 0 0>)
    Force Strength=&Multiply_Float(
     A=-2000
     B=&FloatFromCurve(FloatCurve=DataInterface() {propertyValues {Curve {keys [<at 0 1>,<at 1 2>]}}}, CurveIndex=$Emitter.NormalizedLoopAge, Scale Curve=1.8)
    )
    Module.Use Falloff=false
    Module.UseLinearForce=false
   )
   &ScaleColor(
    Scale Alpha=&FloatFromCurve(FloatCurve=DataInterface() {propertyValues {Curve {keys [<at 0 1>,<at 0.7 1>,<at 0.8 0>]}}})
    Scale RGB=&MakeVectorFromLinearColorRGB(Color=<rgba 0.6099 0.06 3 1>)
    ScaleA=true
    ScaleRGB=true
   )
   /SetVariables(Particles.Glow03_Size=&Multiply_Vector2D_ByFloat(
     Float=&FloatFromCurve(
      FloatCurve=DataInterface() {propertyValues {Curve {keys [
          <at 0 0 RCIM_Cubic>
          <at 0.2 1.3 RCIM_Cubic>
          {arriveTangent=-0.1423, interpMode=RCIM_Cubic, leaveTangent=-0.1423, tangentMode=RCTM_User, time=1, value=0}
         ]}}}
      Scale Curve=7
     )
     Vector2D=<1.5 2>
    )) {isSetParameters=true}
   /SetVariables(Particles.Glow03_Color=&MakeLinearColorFromVectorAndFloat(
     "Vector (RGB)"=&MakeVectorFromLinearColorRGB(Color=&MultiplyLinearColor(Linear Color=<rgba 0.2963 0.097 1 1>, Scale Factor=&MakeVector4_FromFloat(Float=1)))
     "Float (Alpha)"=&Multiply_Float(A=0.2, B=1)
    )) {isSetParameters=true}
   &SolveForcesAndVelocity(Clamp Velocity=false, Limit Acceleration=false)
  ]
  rendererClasses [&NiagaraSpriteRendererProperties]
  renderers [
   {rendererClass=&NiagaraSpriteRendererProperties}
   {rendererClass=&NiagaraSpriteRendererProperties, rendererIndex=1}
  ]
 }
 Ball {
  emitterUpdateScript [
   &EmitterState(Inactive Response=::NewEnumerator1, Life Cycle Mode=::NewEnumerator1, Loop Behavior=::NewEnumerator1, Loop Duration=5.6, UseLoopDelay=true)
   &SpawnBurst_Instantaneous(Use Spawn Probability=false, UseLoopCountLimit=false)
  ]
  particleSpawnScript [
   &InitializeParticle(
    Color=<rgba 0.2143 0 1 1.5>
    Lifetime=5.6
    Mesh Scale Mode=::NewEnumerator3
    Sprite Size Mode=::NewEnumerator3
    Uniform Sprite Size=120
    UsePositionOffset=true
    Write Material Random=true
   )
   /SetVariables(Particles.Wave_SpriteRotation:NiagaraFloat=&RandomRangeFloat(Minimum=0, Maximum=360)) {isSetParameters=true}
  ]
  particleUpdateScript [
   &ParticleState()
   &ScaleSpriteSize(Uniform Curve Sprite Scale=DataInterface() {propertyValues {Curve {keys [<at 0 0>,<at 0.1 0.5>,<at 0.2 1.375>,<at 0.2986 2.7028>,<at 0.4602 2.7115>,<at 0.48 3.0805>,<at 0.5015 2.4821>,<at 0.5315 3.5201>,<at 0.5551 2>,<at 0.5751 3.2702>,<at 0.6 1.5>,<at 0.6995 2.0605>,<at 0.7372 2.2244>,<at 0.963 2.5493>,<at 1.0042 3.9722>]}}}) {module=ScaleSpriteSize001}
   &ScaleColor(
    Scale Alpha=&FloatFromCurve(FloatCurve=DataInterface() {propertyValues {Curve {keys [{arriveTangent=2.1667, time=0, value=1.5}, <at 1 1.5>]}}})
    Scale RGB=&MakeVectorFromLinearColorRGB(Color=<rgba 1.5 1.5 1.5 1>)
    ScaleA=true
    ScaleRGB=true
   )
   /SetVariables(Particles.Dark_Color:LinearColor=&MakeLinearColorFromVectorAndFloat(
     "Vector (RGB)"=&MakeVectorFromLinearColorRGB(Color=&ColorFromCurve(ColorCurve=DataInterface() {propertyValues {
         AlphaCurve {keys []}
         BlueCurve {keys [<at 0.4617 0.1 RCIM_Cubic>,<at 0.5195 0.02 RCIM_Cubic>]}
         GreenCurve {keys [<at 0.4617 0.005 RCIM_Cubic>,<at 0.5195 0.001 RCIM_Cubic>]}
         RedCurve {keys [<at 0.4617 0.0254 RCIM_Cubic>,<at 0.5195 0.0051 RCIM_Cubic>]}
        }}))
     "Float (Alpha)"=&FloatFromCurve(FloatCurve=DataInterface() {propertyValues {Curve {keys [<at 0.4524 0.9974>,<at 0.5242 0.92>]}}})
    )) {isSetParameters=true}
   /SetVariables(Particles.Dark_Size:Vector2f=&Multiply_Vector2D(A=$Particles.SpriteSize, B=&Vector2DFromFloat(Value=0.25))) {isSetParameters=true}
   /SetVariables(Particles.Energy_Scale:Vector3f=&MakeVector(
     X=&Multiply_Float(
      A=0.4
      B=&FloatFromCurve(
       FloatCurve=DataInterface() {propertyValues {Curve {keys [<at 0 0>,<at 0.105 0.7849>,<at 0.2054 1.7633>,<at 0.2986 2.7028>,<at 0.4602 2.7115>,<at 0.4765 3.3667>,<at 0.4885 2.2422>,<at 0.5018 3.8877>,<at 0.5196 2.1079>,<at 0.5379 3.4236>,<at 0.5627 1.8811>,<at 0.5781 3.3319>,<at 0.6 1.5>,<at 0.6995 2.0605>,<at 0.7372 2.2244>,<at 0.963 2.8>,<at 1.0042 3.9722>]}}}
       Scale Curve=&RandomRangeFloat(
        Minimum=&FloatFromCurve(FloatCurve=DataInterface() {propertyValues {Curve {keys [<at -0.0011 0.754>,<at 0.1215 0.7809>,<at 0.3 1>,<at 0.3939 1>,<at 0.4447 0.9115>,<at 0.5758 0.9115>,<at 0.6244 1>,<at 0.8 1>,<at 0.857 0.8678>,<at 0.9446 0.865>,<at 1 1>]}}})
        Maximum=1
        Evaluation Type=::NewEnumerator1
       )
      )
     )
     Y=&Multiply_Float(
      A=0.4
      B=&FloatFromCurve(
       FloatCurve=DataInterface() {propertyValues {Curve {keys [<at 0 0>,<at 0.105 0.7849>,<at 0.2054 1.7633>,<at 0.2986 2.7028>,<at 0.4602 2.7115>,<at 0.4779 3.4163>,<at 0.4973 2.1463>,<at 0.5158 3.9425>,<at 0.5312 2.4274>,<at 0.5435 3.7269>,<at 0.5575 2.1051>,<at 0.5732 3.1661>,<at 0.6 1.5>,<at 0.7007 2.2596>,<at 0.7451 2.3523>,<at 0.963 2.8>,<at 1.0042 3.9722>]}}}
       Scale Curve=&RandomRangeFloat(
        Minimum=&FloatFromCurve(FloatCurve=DataInterface() {propertyValues {Curve {keys [<at 0.0008 0.7247>,<at 0.1234 0.7516>,<at 0.3 1>,<at 0.3939 1>,<at 0.4447 0.9115>,<at 0.5758 0.9115>,<at 0.6244 1>,<at 0.8 1>,<at 0.8581 0.8911>,<at 0.9457 0.8971>,<at 1 1>]}}})
        Maximum=1
        Evaluation Type=::NewEnumerator1
       )
      )
     )
     Z=&Multiply_Float(
      A=0.4
      B=&FloatFromCurve(
       FloatCurve=DataInterface() {propertyValues {Curve {keys [<at 0 0>,<at 0.105 0.7849>,<at 0.2054 1.7633>,<at 0.2986 2.7028>,<at 0.4336 2.7101>,<at 0.4531 2.2894>,<at 0.47 3.1424>,<at 0.4788 1.8678>,<at 0.5013 3.6415>,<at 0.5196 2.0873>,<at 0.5402 3.2844>,<at 0.5583 1.8974>,<at 0.5774 2.7419>,<at 0.6 1.5>,<at 0.6994 2.0794>,<at 0.7493 2.2562>,<at 0.963 2.8>,<at 1.0042 3.9722>]}}}
       Scale Curve=&RandomRangeFloat(
        Minimum=&FloatFromCurve(FloatCurve=DataInterface() {propertyValues {Curve {keys [<at 0.0034 0.7972>,<at 0.1261 0.8241>,<at 0.3 1>,<at 0.3939 1>,<at 0.4472 0.867>,<at 0.5783 0.867>,<at 0.6244 1>,<at 0.8 1>,<at 0.8525 0.8714>,<at 0.9525 0.8714>,<at 1 1>]}}})
        Maximum=1
        Evaluation Type=::NewEnumerator1
       )
      )
     )
    )) {isSetParameters=true}
   /SetVariables(Particles.Energy_Color:LinearColor=&MakeLinearColorFromVectorAndFloat(
     "Vector (RGB)"=&MakeVectorFromLinearColorRGB(Color=&ColorFromCurve(ColorCurve=DataInterface() {propertyValues {
         AlphaCurve {keys []}
         BlueCurve {keys [<at 0.4727 0.25 RCIM_Cubic>,<at 0.5367 0.35 RCIM_Cubic>]}
         GreenCurve {keys [<at 0.4727 0.0333 RCIM_Cubic>,<at 0.5367 0.0466 RCIM_Cubic>]}
         RedCurve {keys [<at 0.4727 0.1008 RCIM_Cubic>,<at 0.5367 0.1411 RCIM_Cubic>]}
        }}))
     "Float (Alpha)"=1
    )) {isSetParameters=true}
   /SetVariables(Particles.Wave_Color:LinearColor=&MakeLinearColorFromVectorAndFloat(
     "Vector (RGB)"=&MakeVectorFromLinearColorRGB(Color=&MakeCustomLinearColorFromBool(Bool=$User.bIsFrontEndPreview, True Color=<rgba 2.4375 0.3 6 1>, False Color=<rgba 3.8125 0.1 10 1>))
     "Float (Alpha)"=&FloatFromCurve(FloatCurve=DataInterface() {propertyValues {Curve {keys [<at 0.55 1>,<at 0.6 0.3>,<at 1 0.3>]}}})
    )) {isSetParameters=true}
   /SetVariables(Particles.Wave_Size:Vector2f=&Vector2DFromFloat(Value=&Multiply_Float(
      A=22
      B=&FloatFromCurve(FloatCurve=DataInterface() {propertyValues {Curve {keys [<at 0 0>,<at 0.1 0.5>,<at 0.2 1.375>,<at 0.2986 2.7028>,<at 0.4602 2.7115>,<at 0.48 3.0805>,<at 0.5015 2.4821>,<at 0.5315 3.5201>,<at 0.5551 2>,<at 0.5751 3.2702>,<at 0.6 1.5>,<at 0.6995 2.0605>,<at 0.7372 2.2244>,<at 0.963 2.5>,<at 1.0042 3.9722>]}}})
     ))) {isSetParameters=true}
   /SetVariables(Particles.Ring_Size:Vector2f=&Vector2DFromFloat(Value=&Multiply_Float(
      A=25
      B=&FloatFromCurve(FloatCurve=DataInterface() {propertyValues {Curve {keys [<at 0 0>,<at 0.1 0.75>,<at 0.2 1.5>,<at 0.2986 2.7028>,<at 0.4602 2.7115>,<at 0.48 3.0805>,<at 0.5015 2.4821>,<at 0.5315 3.5201>,<at 0.5551 2>,<at 0.5751 3.2702>,<at 0.6 1.5>,<at 0.7 2.25>,<at 0.9 2.25>,<at 1 3.625>]}}})
     ))) {isSetParameters=true}
   /SetVariables(Particles.Ring_Color:LinearColor=&MakeLinearColorFromVectorAndFloat(
     "Vector (RGB)"=&MakeVectorFromLinearColorRGB(Color=&ColorFromCurve(ColorCurve=DataInterface() {propertyValues {
         AlphaCurve {keys []}
         BlueCurve {keys [<at 0.4457 0.7 RCIM_Cubic>,<at 0.5514 0.6 RCIM_Cubic>]}
         GreenCurve {keys [<at 0.4457 0.126 RCIM_Cubic>,<at 0.5514 0.108 RCIM_Cubic>]}
         RedCurve {keys [<at 0.4457 0.3425 RCIM_Cubic>,<at 0.5514 0.2963 RCIM_Cubic>]}
        }}))
     "Float (Alpha)"=&FloatFromCurve(FloatCurve=DataInterface() {propertyValues {Curve {keys [<at 0 0.5>,<at 0.5 0.5>,<at 0.6 0.9>,<at 0.8 1>]}}})
    )) {isSetParameters=true}
   /SetVariables(Particles.Wave_Detail_Color:LinearColor=&MakeLinearColorFromVectorAndFloat(
     "Vector (RGB)"=&MakeVectorFromLinearColorRGB(Color=<rgba 0.7625 0.02 2 1>)
     "Float (Alpha)"=&FloatFromCurve(FloatCurve=DataInterface() {propertyValues {Curve {keys [<at 0 0>,<at 0.38 0>,<at 0.42 1>,<at 1 1>]}}})
    )) {isSetParameters=true}
   /SetVariables(Particles.Wave_Detail_Scale:Vector3f=&Multiply_Vector(
     A=$Particles.Energy_Scale
     B=&VectorFromFloat(Value=&Multiply_Float(
       A=0.3
       B=&FloatFromCurve(FloatCurve=DataInterface() {propertyValues {Curve {keys [<at 0.38 1>,<at 0.4 1>]}}})
      ))
    )) {isSetParameters=true}
   /SetVariables(Particles.Wave_Detail_Dyn:Vector4f=&MakeVector4(
     X=&FloatFromCurve(FloatCurve=DataInterface() {propertyValues {Curve {keys [<at 0.5 0>,<at 1 1>]}}}, Scale Curve=6)
     Y=0
     Z=0
     W=0
    )) {isSetParameters=true}
   /SetVariables(Particles.Wave_Detail_Center_Scale:Vector3f=&VectorFromFloat(Value=&Multiply_Float(
      A=0.2
      B=&FloatFromCurve(FloatCurve=DataInterface() {propertyValues {Curve {keys [<at 0 0 RCIM_Cubic>, {arriveTangent=4.2303, interpMode=RCIM_Cubic, leaveTangent=4.2303, tangentMode=RCTM_User, time=0.3, value=1}]}}})
     ))) {isSetParameters=true}
   /SetVariables(Particles.Wave_Detail_Center_Color:LinearColor=&MakeLinearColorFromVectorAndFloat("Vector (RGB)"=&MakeVectorFromLinearColorRGB(Color=<rgba 0.08 0.0021 0.2 1>), "Float (Alpha)"=1)) {isSetParameters=true}
   &FN_Light_Attributes(
    Color=<rgba 11.0771 3 30 1>
    Create Color Attribute=true
    Create Exponent Attribute=true
    Create Position Attribute=true
    Create Radius Attribute=true
    Enable Particle Light Enable=true
    Exponent=2
    Radius=&FloatFromCurve(FloatCurve=DataInterface() {propertyValues {Curve {keys [<at 0 0.2>,<at 0.2 1>,<at 1 1>]}}}, Scale Curve=120)
    Scale Color=&FloatFromCurve(FloatCurve=DataInterface() {propertyValues {Curve {keys [<at 0 0>,<at 0.1 1>,<at 0.9 1>,<at 1 0>]}}})
   )
  ]
  rendererClasses [
   &NiagaraSpriteRendererProperties
   &NiagaraMeshRendererProperties
   &NiagaraLightRendererProperties
  ]
  renderers [
   {rendererClass=&NiagaraSpriteRendererProperties}
   {rendererClass=&NiagaraMeshRendererProperties, rendererIndex=1}
   {rendererClass=&NiagaraSpriteRendererProperties, rendererIndex=2}
   {rendererClass=&NiagaraSpriteRendererProperties, rendererIndex=3}
   {rendererClass=&NiagaraMeshRendererProperties, rendererIndex=4}
   {rendererClass=&NiagaraMeshRendererProperties, rendererIndex=5}
   {rendererClass=&NiagaraLightRendererProperties, rendererIndex=6}
  ]
 }
 Center_Flare {
  emitterUpdateScript [
   &EmitterState(Inactive Response=::NewEnumerator1, Life Cycle Mode=::NewEnumerator1, Loop Behavior=::NewEnumerator1, Loop Duration=5.6, UseLoopDelay=true)
   &SpawnBurst_Instantaneous(Use Spawn Probability=false, UseLoopCountLimit=false)
  ]
  particleSpawnScript [&InitializeParticle(Lifetime=5.6, Sprite Size=<45 25>, Sprite Size Mode=::NewEnumerator4, UsePositionOffset=false, Write Material Random=true)]
  particleUpdateScript [
   &ParticleState()
   &ScaleColor(
    Scale Alpha=&FloatFromCurve(
     FloatCurve=DataInterface() {propertyValues {Curve {keys [<at 0 0>,<at 0.1 1>,<at 1 1>]}}}
     Scale Curve=&Multiply_Float(A=1, B=&FNTimeOfDayScalar(Night Scale=0.6, Morning Scale=1, Evening Scale=1))
    )
    Scale RGB=&MakeVectorFromLinearColorRGB(Color=&ColorFromCurve(ColorCurve=DataInterface() {propertyValues {
        AlphaCurve {keys []}
        BlueCurve {keys [<at 0 2 RCIM_Cubic>,<at 0.7819 2>,<at 1 5 RCIM_Cubic>]}
        GreenCurve {keys [<at 0 0.2 RCIM_Cubic>,<at 0.7819 0.2>,<at 1 0.4 RCIM_Cubic>]}
        RedCurve {keys [<at 0 0.9691 RCIM_Cubic>,<at 0.7819 0.9501>,<at 1 1.7761 RCIM_Cubic>]}
       }}))
    ScaleA=true
    ScaleRGB=true
   )
   &ScaleSpriteSize(Uniform Curve Sprite Scale=DataInterface() {propertyValues {Curve {keys [{tangentMode=RCTM_User, time=0, value=0}, {tangentMode=RCTM_User, time=0.1027, value=1}]}}}) {enabled=false}
   &ScaleSpriteSize(
    Uniform Curve Scale=&RandomRangeFloat(
     Minimum=&FloatFromCurve(FloatCurve=DataInterface() {propertyValues {Curve {keys [<at -0.0014 0.3486>,<at 0.1888 0.3775>,<at 0.3052 0.6995>,<at 0.3991 0.6995>,<at 0.4636 0.313>,<at 0.6358 0.315>,<at 0.6895 0.7033>,<at 0.8313 0.7033>,<at 0.8651 0.2002>,<at 0.9561 0.2001>,<at 1 1>]}}})
     Maximum=1
     Evaluation Type=::NewEnumerator1
    )
    Uniform Curve Sprite Scale=DataInterface() {propertyValues {Curve {keys [<at 0 1>,<at 0.1 1>,<at 0.2785 1.7544>,<at 0.3989 2.8361>,<at 0.5124 2.8422>,<at 0.5268 3.225>,<at 0.5582 2.6669>,<at 0.5781 3.036>,<at 0.5996 2.4375>,<at 0.6267 3.0934>,<at 0.6531 1.9554>,<at 0.6731 3.2257>,<at 0.6981 1.4554>,<at 0.73 2.22>,<at 0.93 2.7>,<at 1 4>]}}}
   ) {module=ScaleSpriteSize001}
   &SolveForcesAndVelocity(Clamp Velocity=false, Limit Acceleration=false)
  ]
  rendererClasses [&NiagaraSpriteRendererProperties]
  renderers [{rendererClass=&NiagaraSpriteRendererProperties}]
 }
 Lightning {
  emitterUpdateScript [
   &EmitterState(
    Inactive Response=::NewEnumerator1
    Life Cycle Mode=::NewEnumerator1
    Loop Behavior=::NewEnumerator1
    Loop Delay=3
    Loop Duration=2.6
    UseLoopDelay=true
   )
   &SpawnBurst_Instantaneous(Use Spawn Probability=false, UseLoopCountLimit=false)
   &SpawnRate(SpawnRate=30, Use Spawn Probability=true)
  ]
  particleSpawnScript [
   &InitializeParticle(
    Color=<rgba 1.496 0.2099 5 1>
    Lifetime Max=0.2
    Lifetime Min=0.15
    Lifetime Mode=::NewEnumerator1
    Mesh Scale=&RandomRangeVector(Minimum=<1 1 1>, Maximum=<1.5 1.5 1.5>)
    Mesh Scale Mode=::NewEnumerator4
    UsePositionOffset=true
    Write Material Random=true
   )
   &CameraOffset(Camera Offset Amount=5)
   &InitialMeshOrientation(
    Mesh Orientation Mode=::NewEnumerator2
    Rotation=&RandomRangeVector(Minimum=<-1 -1 -1>, Maximum=<1 1 1>)
    Use Rotation Vector=true
   )
   &AddVelocity(Module.Use Falloff Distance=false, Override Seed=false, Velocity Mode=::NewEnumerator1, Velocity Speed=1)
   /SetVariables(Particles.Random=&RandomRangeFloat(Minimum=0, Maximum=1)) {isSetParameters=true}
  ]
  particleUpdateScript [
   &ParticleState()
   &ScaleMeshSize(Scale Factor=&Multiply_VectorByFloat(Vector=<1 1 1>, Float=&RandomRangeFloat(Minimum=1, Maximum=1.2)))
   &ScaleMeshSize(Scale Factor=&VectorFromFloat(Value=&FloatFromCurve(FloatCurve=DataInterface() {propertyValues {Curve {keys [<at 0 1>,<at 4 1>,<at 6 1.5>]}}}, CurveIndex=$Emitter.Age))) {module=ScaleMeshSize001}
   &ScaleColor(
    Scale Alpha=&Multiply_Float(
     A=10
     B=&FloatFromCurve(FloatCurve=DataInterface() {propertyValues {Curve {keys [
          <at 0 0 RCIM_Cubic>
          {arriveTangent=-0.0166, interpMode=RCIM_Cubic, leaveTangent=-0.0166, tangentMode=RCTM_User, time=0.1, value=0}
          <at 0.2 1 RCIM_Cubic>
          <at 1 0 RCIM_Cubic>
         ]}}})
    )
    Scale RGB=&Multiply_VectorByFloat(
     Vector=<1 1 1>
     Float=&Multiply_Float(A=&RandomRangeFloat(Minimum=0, Maximum=1, Evaluation Type=::NewEnumerator1), B=5)
    )
    ScaleA=true
    ScaleRGB=true
   )
   &DynamicMaterialParameters(
    Index 0 Param 1=&FloatFromCurve(FloatCurve=DataInterface() {propertyValues {Curve {keys [{tangentMode=RCTM_User, time=0, value=0}, {tangentMode=RCTM_User, time=1, value=1}]}}})
    Index 0 Param 2=0.7
    Index 0 Param 3=&FloatFromCurve(FloatCurve=DataInterface() {propertyValues {Curve {keys [<at 0 0 RCIM_Cubic>, {arriveTangent=-0.3684, interpMode=RCIM_Cubic, leaveTangent=-0.3684, tangentMode=RCTM_User, time=1, value=3}]}}})
    Index 0 Param 4=1
    Index 1 Param 1=1
    Index 1 Param 3=&Lerp_Float(A=1.3, B=1.8, Alpha=$Particles.Random)
    Index 1 Param 4=1
    Module.Param0WriteEnabled=true
    Module.Param0WriteEnabled1=true
    Module.Param1WriteEnabled=true
    Module.Param1WriteEnabled1=true
    Module.Param2WriteEnabled=true
    Module.Param2WriteEnabled1=true
    Module.Param3WriteEnabled=true
    Module.Param3WriteEnabled1=true
    Write Parameter Index 1=true
   )
   &SolveForcesAndVelocity(Clamp Velocity=false, Limit Acceleration=false)
   /SetVariables(Particles.Lightning glow color:LinearColor=&MakeLinearColorFromVectorAndFloat(
     "Vector (RGB)"=&MakeVectorFromLinearColorRGB(Color=&MultiplyLinearColor(
       Linear Color=<rgba 0.5 0 2 1>
       Scale Factor=/Niagara/DynamicInputs/Multiply/Multiply_Vector4_ByFloat(Vector4:Vector4f=<1 1 1 1>, Float:NiagaraFloat=1)
      ))
     "Float (Alpha)"=&Multiply_Float(
      A=&FloatFromCurve(FloatCurve=DataInterface() {propertyValues {Curve {keys [
           {arriveTangent=-0.6023, interpMode=RCIM_Cubic, leaveTangent=-0.6023, tangentMode=RCTM_User, time=0, value=0}
           <at 0.1 0.85 RCIM_Cubic>
           {arriveTangent=-0.4241, interpMode=RCIM_Cubic, leaveTangent=-0.4241, tangentMode=RCTM_User, time=0.8, value=0}
          ]}}})
      B=3
     )
    )) {isSetParameters=true}
   /SetVariables(
    Particles.Glow Dy02:Vector4f=&MakeVector4(
     X=&FloatFromCurve(FloatCurve=DataInterface() {propertyValues {Curve {keys [{tangentMode=RCTM_User, time=0, value=0.2}, {tangentMode=RCTM_User, time=0.5, value=1}]}}})
     Y=0.7
     Z=0
     W=1
    )
    Particles.Glow Dy03:Vector4f=&MakeVector4(
     X=&FloatFromCurve(FloatCurve=DataInterface() {propertyValues {Curve {keys [<at 0.5 1>,<at 1 0>]}}}, Scale Curve=0.5)
     Y=0
     Z=&Lerp_Float(A=1.3, B=1.8, Alpha=$Particles.Random)
     W=1
    )
   ) {isSetParameters=true}
  ]
  rendererClasses [&NiagaraMeshRendererProperties]
  renderers [
   {rendererClass=&NiagaraMeshRendererProperties, rendererIndex=0}
   {rendererClass=&NiagaraMeshRendererProperties, rendererIndex=1}
  ]
 }
 Smoke_Ring {
  emitterUpdateScript [
   &EmitterState(
    Inactive Response=::NewEnumerator1
    Life Cycle Mode=::NewEnumerator1
    Loop Behavior=::NewEnumerator1
    Loop Delay=1.8
    Loop Duration=4
    UseLoopDelay=true
   )
   &SpawnRate(SpawnRate=10, Use Spawn Probability=false)
   &SpawnBurst_Instantaneous(Spawn Count=3, Use Spawn Probability=false, UseLoopCountLimit=false)
  ]
  particleSpawnScript [
   &InitializeParticle(
    Color Minimum=<rgba 0.3 0.3 0.3 1>
    Color Mode=::NewEnumerator2
    Lifetime Max=1
    Lifetime Min=0.5
    Lifetime Mode=::NewEnumerator1
    Mesh Scale Max=<1 1 15>
    Mesh Scale Min=<1 1 10>
    Mesh Scale Mode=::NewEnumerator2
    UsePositionOffset=false
    Write Material Random=true
   ) {module=InitializeParticle001}
   &InitialMeshOrientation(
    Mesh Orientation Mode=::NewEnumerator2
    Rotation=&RandomRangeVector(Minimum=<0 0 0>, Maximum=<1 1 1>)
    Use Rotation Vector=true
   )
  ]
  particleUpdateScript [
   &ParticleState()
   &UpdateMeshOrientation(Rotation Rate=&RandomRangeFloat(Minimum=-1.5, Maximum=-2), Rotation Vector=<0 0 1>)
   &ScaleMeshSize(Scale Factor=&Multiply_VectorByFloat(
     Vector=&Multiply_VectorByFloat(
      Vector=&MakeVector(
       X=1
       Y=1
       Z=&Multiply_Float(
        A=&FloatFromCurve(
         FloatCurve=DataInterface() {propertyValues {Curve {keys [{arriveTangent=-1.572, time=0, value=1}, {interpMode=RCIM_Cubic, tangentMode=RCTM_User, time=1, value=0.5}]}}}
         Scale Curve=0.7
        )
        B=&RandomRangeFloat(Minimum=1, Maximum=1)
       )
      )
      Float=&RandomRangeFloat(Minimum=0.8, Maximum=1)
     )
     Float=&FloatFromCurve(
      FloatCurve=DataInterface() {propertyValues {Curve {keys [{arriveTangent=2.2632, time=0, value=1}, {arriveTangent=2.9967, leaveTangent=2.9967, time=1, value=1}]}}}
      Scale Curve=&Multiply_Float(A=&RandomRangeFloat(Minimum=2, Maximum=3), B=1)
     )
    ))
   &ScaleColor(
    Scale Alpha=&FloatFromCurve(
     FloatCurve=DataInterface() {propertyValues {Curve {keys [
         {arriveTangent=6.8924, interpMode=RCIM_Cubic, leaveTangent=6.8924, tangentMode=RCTM_User, time=0.02, value=0}
         {arriveTangent=1.141, interpMode=RCIM_Cubic, leaveTangent=1.141, tangentMode=RCTM_User, time=0.1, value=0.6}
         {arriveTangent=-1.0842, interpMode=RCIM_Cubic, leaveTangent=-1.0842, tangentMode=RCTM_User, time=0.6, value=0.55}
         {arriveTangent=-0.2428, interpMode=RCIM_Cubic, leaveTangent=-0.2428, tangentMode=RCTM_User, time=1, value=0}
        ]}}}
     Scale Curve=&Multiply_Float(
      A=&Multiply_Float(
       A=&RandomRangeFloat(Minimum=1.5, Maximum=2)
       B=&FNTimeOfDayScalar(Night Scale=0.5, Morning Scale=1, Evening Scale=1)
      )
      B=1
     )
    )
    Scale RGB=&MakeVectorFromLinearColorRGB(Color=&ColorFromCurve(ColorCurve=DataInterface() {propertyValues {
        AlphaCurve {keys []}
        BlueCurve {keys [<at 0.0271 0.6>,<at 0.4753 0.05 RCIM_Cubic>]}
        GreenCurve {keys [<at 0.0271 0.0374>,<at 0.4753 0 RCIM_Cubic>]}
        RedCurve {keys [<at 0.0271 0.2497>,<at 0.4753 0.0107 RCIM_Cubic>]}
       }}))
    ScaleA=true
    ScaleRGB=true
   )
  ]
  rendererClasses [&NiagaraMeshRendererProperties]
  renderers [{rendererClass=&NiagaraMeshRendererProperties, rendererIndex=0}]
 }
 Trail {
  emitterUpdateScript [
   &EmitterState(Inactive Response=::NewEnumerator1, Life Cycle Mode=::NewEnumerator1, Loop Behavior=::NewEnumerator1, Loop Duration=4.3, UseLoopDelay=false)
   &SpawnRate(SpawnRate=45, Use Spawn Probability=false) {enabled=false}
   &SpawnPerUnit(
    Module.Use Max Movement Threshold=true
    Module.Use Movement Tolerance=false
    Spawn Probability:NiagaraFloat=&FloatFromCurve(FloatCurve=DataInterface() {propertyValues {Curve {keys [<at 0.8 1>,<at 0.85 0>,<at 1 0>,<at 1.0322 1.0018>]}}}, CurveIndex=$Emitter.Age)
    Spawn Spacing=3
    Use Spawn Probability=true
   )
  ]
  particleSpawnScript [&InitializeRibbon(
    Color=<rgba 1 1 1 2>
    Lifetime=0.35
    Ribbon Width=&Multiply_Float(
     A=20
     B=&FloatFromCurve(FloatCurve=DataInterface() {propertyValues {Curve {keys [<at 0 1>,<at 0.9 1>,<at 1.1 2>,<at 3 2>]}}}, CurveIndex=$Emitter.Age)
    )
    UsePositionOffset=false
    Write Material Random=true
   )]
  particleUpdateScript [
   &ParticleState()
   &ScaleRibbonWidth(Ribbon Width Scale=&FloatFromCurve(FloatCurve=DataInterface() {propertyValues {Curve {keys [
         {arriveTangent=10.6409, interpMode=RCIM_Cubic, leaveTangent=7.4546, tangentMode=RCTM_Break, time=0, value=0.5}
         {interpMode=RCIM_Cubic, leaveTangent=0.0729, tangentMode=RCTM_Break, time=0.1, value=1}
         {arriveTangent=-0.3829, interpMode=RCIM_Cubic, leaveTangent=-0.3829, tangentMode=RCTM_User, time=0.3, value=1}
         {arriveTangent=-2.0489, interpMode=RCIM_Cubic, leaveTangent=-2.0489, tangentMode=RCTM_User, time=1, value=0}
        ]}}}))
   &ScaleColor(
    Linear Color Curve=DataInterface() {propertyValues {
      AlphaCurve {keys [<at 0 1>,<at 0.2985 1>,<at 0.5735 0.4>,<at 1 0>]}
      BlueCurve {keys [<at 0 0>,<at 0.1031 0>,<at 0.3817 0.1139>,<at 1 0.0677>]}
      GreenCurve {keys [<at 0 2>,<at 0.1031 0.8438>,<at 0.3817 0.3021>,<at 1 0.1458>]}
      RedCurve {keys [<at 0 0.3354>,<at 0.1031 0.0304>,<at 0.3817 0>,<at 1 0.0017>]}
     }}
    Scale Mode=::NewEnumerator2
   ) {enabled=false, module=ScaleColor001}
   &ScaleColor(
    Scale Alpha=&FloatFromCurve(FloatCurve=DataInterface() {propertyValues {Curve {keys [<at 0 1>,<at 0.6 0.45>,<at 0.8 0.15>,<at 1 0>]}}})
    Scale RGB=&MakeVectorFromLinearColorRGB(Color=&ColorFromCurve(ColorCurve=DataInterface() {propertyValues {
        AlphaCurve {keys []}
        BlueCurve {keys [<at 0 2>,<at 0.1654 0.8438>,<at 0.6695 0.3021>]}
        GreenCurve {keys [<at 0 0.3357>,<at 0.1654 0.0848>,<at 0.6695 0>]}
        RedCurve {keys [<at 0 1.2825>,<at 0.1654 0.3592>,<at 0.6695 0.0813>]}
       }}))
    ScaleA=true
    ScaleRGB=true
   ) {module=ScaleColor002}
  ]
  rendererClasses [&NiagaraRibbonRendererProperties]
  renderers [{rendererClass=&NiagaraRibbonRendererProperties}]
 }
 Trail_Spark {
  emitterUpdateScript [
   &EmitterState(
    Inactive Response=::NewEnumerator1
    Life Cycle Mode=::NewEnumerator1
    Loop Behavior=::NewEnumerator1
    Loop Delay=0.085
    Loop Duration=3
    UseLoopDelay=true
   )
   &SpawnRate(
    SpawnRate=&Multiply_Float(
     A=40
     B=&FloatFromCurve(
      FloatCurve=DataInterface() {propertyValues {Curve {keys [<at 0.4268 0.9943>,<at 0.4497 -0.0057>,<at 0.9827 0.0057>,<at 1.0346 2.5>,<at 1.2843 2.5>]}}}
      CurveIndex=$Emitter.Age
     )
    )
    Use Spawn Probability=false
   )
   &SpawnBurst_Instantaneous(Spawn Count=3, Use Spawn Probability=false, UseLoopCountLimit=false) {enabled=false}
  ]
  particleSpawnScript [
   &InitializeParticle(
    Color=<rgba 2.9307 0.2765 10 1>
    Lifetime Max=0.5
    Lifetime Min=0.15
    Lifetime Mode=::NewEnumerator1
    Sprite Size Max=<2.2 6>
    Sprite Size Min=<1.5 3>
    Sprite Size Mode=::NewEnumerator2
    UsePositionOffset=false
    Write Material Random=true
   )
   &ShapeLocation(MaskSpawnGroup=false, Override Seed=false, Sphere Radius=5, UseOwnerScale=true)
   &AddVelocity(Module.Use Falloff Distance=false, Override Seed=false, Velocity Mode=::NewEnumerator1, Velocity Speed=40)
   &InheritVelocity(
    Inherited Velocity Speed Limit:NiagaraFloat=&RandomRangeFloat(Minimum=-20, Maximum=-40)
    Speed Limit=true
    UseSpeedThreshold=false
   )
  ]
  particleUpdateScript [
   &ParticleState()
   &CurlNoiseForce(Bias Noise=false, Noise Strength=50, Pan Noise=false)
   &Drag(Drag=2, Use Linear Drag=true, Use Rotational Drag=false)
   &SolveForcesAndVelocity(Clamp Velocity=false, Limit Acceleration=false)
   &ScaleSpriteSizeBySpeed(Max Scale Factor=<1 2>)
   &ScaleSpriteSize(
    Non-Uniform Curve Sprite Scale=DataInterface() {propertyValues {
      XCurve {keys [{tangentMode=RCTM_User, time=0, value=1}, {tangentMode=RCTM_User, time=1, value=0.2}]}
      YCurve {keys [{tangentMode=RCTM_User, time=0, value=1}, {tangentMode=RCTM_User, time=1, value=0}]}
     }}
    Scale Sprite Size Mode=::NewEnumerator3
   )
   &FN_Sprite_Attributes(
    Color=&MultiplyLinearColor(
     Linear Color=<rgba 0.2963 0.097 1 1>
     Scale Factor=&MakeVector4_FromFloat(Float=&FloatFromCurve(FloatCurve=DataInterface() {propertyValues {Curve {keys [
           {interpMode=RCIM_Cubic, tangentMode=RCTM_User, time=0, value=0}
           {interpMode=RCIM_Cubic, tangentMode=RCTM_User, time=0.5, value=0.8}
           {tangentMode=RCTM_User, time=1, value=0}
          ]}}}))
    )
    Create Color Attribute=true
    Create SpriteSize Attribute=true
    SpriteSize=&Multiply_Vector2D(
     A=&RandomRangeVector2D(Minimum=<3 5>, Maximum=<4 6>)
     B=&Vector2DFromFloat(Value=&FloatFromCurve(
       FloatCurve=DataInterface() {propertyValues {Curve {keys [<at 0 0>,<at 0.1 1>,<at 0.3 0.35>,<at 0.4 0.75>,<at 0.561 0.454>,<at 0.7 0.6>,<at 0.8 0.35>,<at 0.9 0.45>,<at 1 0>]}}}
       Scale Curve=&RandomRangeFloat(Minimum=1.5, Maximum=3)
      ))
    )
    Write to Module Namespace=true
   )
  ]
  rendererClasses [&NiagaraSpriteRendererProperties]
  renderers [
   {rendererClass=&NiagaraSpriteRendererProperties}
   {rendererClass=&NiagaraSpriteRendererProperties, rendererIndex=1}
  ]
 }
}
summary {
 systemName=NS_Emote_TidalNinja_Ball_R
 emitters [
  {
   emitterName=Ball
   rendererClasses [
    &NiagaraSpriteRendererProperties
    &NiagaraMeshRendererProperties
    &NiagaraLightRendererProperties
   ]
  }
  {emitterName=Absorb_Beam, rendererClasses [&NiagaraMeshRendererProperties]}
  {emitterName=Center_Flare, rendererClasses [&NiagaraSpriteRendererProperties]}
  {emitterName=Smoke_Ring, rendererClasses [&NiagaraMeshRendererProperties]}
  {emitterName=Trail, rendererClasses [&NiagaraRibbonRendererProperties]}
  {emitterName=Trail_Spark, rendererClasses [&NiagaraSpriteRendererProperties]}
  {emitterName=Absorb_Particles, rendererClasses [&NiagaraSpriteRendererProperties]}
  {emitterName=Lightning, rendererClasses [&NiagaraMeshRendererProperties]}
  {emitterName=Absorb_Energy, rendererClasses [&NiagaraMeshRendererProperties]}
 ]
 userVariables [
  User.bIsFrontEnd {type=NiagaraBool}
  User.bIsFrontEndPreview {type=NiagaraBool}
 ]
}
