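# Niagara system for a small purple energy ball effect in the TidalNinja player emote.
# A single Ball emitter burst-spawns randomly sized sprites with four sprite renderers.
# Update-time parameters derive separate size, color or dynamic values for Dark, Light, Wave and Ring layers from the base particle, with randomized Light size.
# A size curve pops in, holds, pops again and collapses early in the lifetime; a second size-scale module is disabled.
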NiagaraSystem /BRCosmetics/Animation/Game/MainPlayer/Emotes/TidalNinja/FX/NS_Emote_TidalNinja_Ball_Small
dependencies {usedDataInterfaces [&NiagaraDataInterfaceCurve], usedDynamicInputs [], usedModules [&SystemState]}
emitters {Ball {
  emitterUpdateScript [
   &EmitterState(Inactive Response=::NewEnumerator1, Life Cycle Mode=::NewEnumerator1, Loop Behavior=::NewEnumerator1, UseLoopDelay=true)
   &SpawnBurst_Instantaneous(Use Spawn Probability=false, UseLoopCountLimit=false)
  ]
  particleSpawnScript [&InitializeParticle(
    Color=<rgba 0.2143 0 1 2>
    Lifetime Max=3
    Lifetime Min=2
    Lifetime Mode=::NewEnumerator1
    Sprite Size Mode=::NewEnumerator1
    Uniform Sprite Size Max=150
    Uniform Sprite Size Min=80
    UsePositionOffset=true
    Write Material Random=true
   )]
  particleUpdateScript [
   &ParticleState()
   &ScaleSpriteSize(Uniform Curve Sprite Scale=DataInterface() {propertyValues {Curve {keys [
        {tangentMode=RCTM_User, time=0, value=0}
        <at 0.0361 2.9756>
        <at 0.076 0.9959>
        <at 0.263 0.9959>
        <at 0.2979 2.4799>
        {tangentMode=RCTM_User, time=0.363, value=0}
       ]}}}) {module=ScaleSpriteSize001}
   &ScaleSpriteSize(
    Uniform Curve Scale=&RandomRangeFloat(
     Minimum=&Lerp_Float(
      A=1
      B=0.85
      Alpha=&FloatFromCurve(FloatCurve=DataInterface() {propertyValues {Curve {keys [<at 0 0>,<at 0.1 1>]}}})
     )
     Maximum=1
     Evaluation Type=::NewEnumerator1
    )
    Uniform Curve Sprite Scale=DataInterface() {propertyValues {Curve {keys [
        <at 0 0>
        <at 0.05 0.6>
        {arriveTangent=17.9642, interpMode=RCIM_Cubic, leaveTangent=17.9642, tangentMode=RCTM_User, time=0.9, value=0.6}
        {arriveTangent=42.1497, interpMode=RCIM_Cubic, leaveTangent=-28.4723, tangentMode=RCTM_Break, time=0.91, value=1}
        <at 0.98 0>
       ]}}}
   ) {enabled=false}
   &ScaleColor(Scale RGB=&MakeVectorFromLinearColorRGB(Color=<rgba 1 1 1 1>), ScaleA=true, ScaleRGB=true)
   /SetVariables(Particles.Dark_Color:LinearColor=&MakeLinearColorFromVectorAndFloat("Vector (RGB)"=&MakeVectorFromLinearColorRGB(Color=<rgba 0.0339 0.0023 0.1 1>), "Float (Alpha)"=3)) {isSetParameters=true}
   /SetVariables(Particles.Dark_Size:Vector2f=&Multiply_Vector2D(A=$Particles.SpriteSize, B=&Vector2DFromFloat(Value=0.3))) {isSetParameters=true}
   /SetVariables(Particles.Dark_Dyn:Vector4f=&MakeVector4(X=5, Y=1, Z=0, W=0)) {isSetParameters=true}
   /SetVariables(Particles.Light_Size:Vector2f=&Multiply_Vector2D(A=$Particles.SpriteSize, B=&RandomRangeVector2D(Minimum=<0.4 0.2>, Maximum=<0.6 0.25>))) {isSetParameters=true}
   /SetVariables(Particles.Wave_Color:LinearColor=&MakeLinearColorFromVectorAndFloat("Vector (RGB)"=&MakeVectorFromLinearColorRGB(Color=<rgba 1.4508 0 10 1>), "Float (Alpha)"=0.7)) {isSetParameters=true}
   /SetVariables(Particles.Wave_Size:Vector2f=&Multiply_Vector2D(A=$Particles.SpriteSize, B=&Vector2DFromFloat(Value=0.1))) {isSetParameters=true}
   /SetVariables(Particles.Ring_Size:Vector2f=&Multiply_Vector2D(A=$Particles.SpriteSize, B=&Vector2DFromFloat(Value=0.1))) {isSetParameters=true}
   /SetVariables(Particles.Ring_Color:LinearColor=&MakeLinearColorFromVectorAndFloat(
     "Vector (RGB)"=&MakeVectorFromLinearColorRGB(Color=<rgba 0.2883 0.1146 0.6 1>)
     "Float (Alpha)"=&FloatFromCurve(FloatCurve=DataInterface() {propertyValues {Curve {keys [<at 0.5 1>,<at 0.7 1>]}}}, Scale Curve=0.5)
    )) {isSetParameters=true}
  ]
  rendererClasses [&NiagaraSpriteRendererProperties]
  renderers [
   {rendererClass=&NiagaraSpriteRendererProperties}
   {rendererClass=&NiagaraSpriteRendererProperties, rendererIndex=1}
   {rendererClass=&NiagaraSpriteRendererProperties, rendererIndex=2}
   {rendererClass=&NiagaraSpriteRendererProperties, rendererIndex=3}
  ]
 }}
summary {
 systemName=NS_Emote_TidalNinja_Ball_Small
 emitters [{emitterName=Ball, rendererClasses [&NiagaraSpriteRendererProperties]}]
 userVariables [
  User.bIsFrontEnd {type=NiagaraBool}
  User.bIsFrontEndPreview {type=NiagaraBool}
 ]
}
