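# Niagara system for the TidalNinja emote's Division effect: a short, purple-violet energy burst in which every particle lives under a second.
# Mesh emitters supply energy leak, four oriented flow meshes and an expanding shockwave, alongside sprite glows and a particle light.
# Fifty flickering sparks spray along opposite X-axis cones under curl noise, drag and upward acceleration.
# A front-end preview flag and time-of-day scalars adjust brightness, alpha and light intensity.
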NiagaraSystem /BRCosmetics/Animation/Game/MainPlayer/Emotes/TidalNinja/FX/NS_Emote_TidalNinja_Division
dependencies {
 usedDataInterfaces [&NiagaraDataInterfaceCurve, &NiagaraDataInterfaceColorCurve, &NiagaraDataInterfaceArrayFloat3]
 usedDynamicInputs [&MakeVector4_FromLinearColor, &Multiply_Vector2D_ByFloat]
 usedModules [&SystemState]
}
emitters {
 EnergyLeak001 {
  emitterUpdateScript [
   &EmitterState(Inactive Response=::NewEnumerator1, Life Cycle Mode=::NewEnumerator1, Loop Behavior=::NewEnumerator1, Loop Duration=1, UseLoopDelay=true)
   &SpawnBurst_Instantaneous(Use Spawn Probability=false, UseLoopCountLimit=false)
  ]
  particleSpawnScript [
   &InitializeParticle(
    Color=<rgba 1.1417 0.6281 2 1>
    Lifetime=0.4
    Mesh Scale Mode=::NewEnumerator3
    Mesh Uniform Scale=1.6
    UsePositionOffset=false
    Write Material Random=true
   )
   &InitialMeshOrientation(
    Mesh Orientation Mode=::NewEnumerator2
    Rotation=&RandomRangeVector(Minimum=<0 0 0>, Maximum=<1 0 0>)
    Use Rotation Vector=true
   )
   &CameraOffset(Camera Offset Amount=2)
  ]
  particleUpdateScript [
   &ParticleState()
   &ScaleColor(
    Scale Alpha=&FloatFromCurve(
     FloatCurve=DataInterface() {propertyValues {Curve {keys [
         {arriveTangent=0.0004, interpMode=RCIM_Cubic, leaveTangent=0.0004, tangentMode=RCTM_User, time=0, value=0}
         {arriveTangent=0.02, interpMode=RCIM_Cubic, leaveTangent=0.02, tangentMode=RCTM_User, time=0.5, value=1}
         {arriveTangent=0.0459, interpMode=RCIM_Cubic, leaveTangent=0.0459, tangentMode=RCTM_User, time=1, value=0}
        ]}}}
     Scale Curve=&RandomRangeFloat(Minimum=0.6, Maximum=1)
    )
    Scale RGB=&MakeVectorFromLinearColorRGB(Color=<rgba 2 2 2 1>)
    ScaleA=true
    ScaleRGB=true
   )
   &ScaleMeshSize(Scale Factor=&VectorFromFloat(Value=&FloatFromCurve(FloatCurve=DataInterface() {propertyValues {Curve {keys [{arriveTangent=2.4731, interpMode=RCIM_Cubic, leaveTangent=2.4731, tangentMode=RCTM_User, time=0, value=0}, <at 1 1 RCIM_Cubic>]}}})))
   &DynamicMaterialParameters(
    Index 0 Param 3=-1
    Module.Index0Float0Write=true
    Module.Index0Float1Write=true
    Module.Index0Float2Write=true
    Module.Index0Float3Write=true
   )
  ]
  rendererClasses [&NiagaraMeshRendererProperties]
  renderers [{rendererClass=&NiagaraMeshRendererProperties, rendererIndex=0}]
 }
 Flow_Mesh_02 {
  emitterUpdateScript [
   &EmitterState(Inactive Response=::NewEnumerator1, Life Cycle Mode=::NewEnumerator1, Loop Behavior=::NewEnumerator1, Loop Duration=1, UseLoopDelay=false)
   &SpawnBurst_Instantaneous(Spawn Count=4, Use Spawn Probability=false)
  ]
  particleSpawnScript [
   &PartitionParticles(Number Of Partitions=4)
   &InitializeParticle(
    Color=<rgba 0.5216 0.1579 1 1>
    Lifetime=0.3
    Mesh Scale=<1.5 1.5 0.5>
    Mesh Scale Mode=::NewEnumerator4
    UsePositionOffset=true
    Write Material Random=true
   )
   &InitialMeshOrientation(
    Rotation=&SelectVectorFromArray(
     Direct Array Index=$Particles.PartitionParticles.Partition
     Vector Selection Array=DataInterface() {propertyValues {FloatData [<0 0.25 0>, <0 0.25 0>, <0 -0.25 0>, <0 -0.25 0>]}}
    )
    Use Orientation Vector=true
    Use Rotation Vector=true
   )
   &InitialMeshOrientation(
    Mesh Orientation Mode=::NewEnumerator2
    Rotation=&RandomRangeVector(Minimum=<0 0 0>, Maximum=<1 0 0>)
    Use Rotation Vector=true
   ) {module=InitialMeshOrientation001}
   &CameraOffset(Camera Offset Amount=2)
  ]
  particleUpdateScript [
   &ParticleState()
   &ScaleColor(
    Scale Alpha=&FloatFromCurve(FloatCurve=DataInterface() {propertyValues {Curve {keys [{arriveTangent=-2.3932, interpMode=RCIM_Cubic, leaveTangent=-2.3932, tangentMode=RCTM_User, time=0, value=1}, <at 1 0 RCIM_Cubic>]}}})
    Scale RGB=&VectorFromFloat(Value=&MakeCustomFloatFromBool(Bool=$User.bIsFrontEndPreview, True Float=3, False Float=4))
    ScaleA=true
    ScaleRGB=true
   )
   &ScaleMeshSize(Scale Factor=&MakeVector(
     X=1
     Y=1
     Z=&FloatFromCurve(FloatCurve=DataInterface() {propertyValues {Curve {keys [
          {arriveTangent=2.515, interpMode=RCIM_Cubic, leaveTangent=2.515, tangentMode=RCTM_User, time=0, value=0}
          {arriveTangent=1.0293, interpMode=RCIM_Cubic, leaveTangent=1.0293, tangentMode=RCTM_User, time=0.3092, value=0.5952}
          {arriveTangent=0.0824, interpMode=RCIM_Cubic, leaveTangent=0.0824, tangentMode=RCTM_User, time=1, value=1}
         ]}}})
    ))
   &SolveForcesAndVelocity(Clamp Velocity=false, Limit Acceleration=false)
  ]
  rendererClasses [&NiagaraMeshRendererProperties]
  renderers [{rendererClass=&NiagaraMeshRendererProperties, rendererIndex=0}]
 }
 Glow {
  emitterUpdateScript [
   &EmitterState(
    Inactive Response=::NewEnumerator1
    Life Cycle Mode=::NewEnumerator1
    Loop Behavior=::NewEnumerator1
    Loop Delay=1.6
    Loop Duration=0.5
    UseLoopDelay=false
   )
   &SpawnBurst_Instantaneous(Use Spawn Probability=false, UseLoopCountLimit=false)
  ]
  particleSpawnScript [
   &InitializeParticle(
    Color=<rgba 0.809 0.1924 3 1>
    Lifetime=0.35
    Sprite Size=<60 40>
    Sprite Size Mode=::NewEnumerator4
    UsePositionOffset=false
    Write Material Random=true
   )
   &SystemLocation(MaskSpawnGroup=false)
   &CameraOffset(Camera Offset Amount=2)
  ]
  particleUpdateScript [
   &ParticleState()
   &ScaleSpriteSize(Uniform Curve Sprite Scale=DataInterface() {propertyValues {Curve {keys [
        {arriveTangent=5.1246, interpMode=RCIM_Cubic, leaveTangent=5.1246, tangentMode=RCTM_User, time=0, value=0}
        {arriveTangent=1.7949, interpMode=RCIM_Cubic, leaveTangent=1.7949, tangentMode=RCTM_User, time=0.3, value=1.5}
        <at 1 2 RCIM_Cubic>
       ]}}})
   &ScaleColor(
    Scale Alpha=&FloatFromCurve(
     FloatCurve=DataInterface() {propertyValues {Curve {keys [
         {arriveTangent=-2.7027, interpMode=RCIM_Cubic, leaveTangent=-2.7027, tangentMode=RCTM_User, time=0, value=1}
         {arriveTangent=-1, interpMode=RCIM_Cubic, leaveTangent=-1, time=0.312, value=0.3309}
         <at 1 0>
        ]}}}
     Scale Curve=&Multiply_Float(A=1, B=&FNTimeOfDayScalar(Night Scale=0.8, Morning Scale=1, Evening Scale=1))
    )
    ScaleA=true
    ScaleRGB=true
   )
   &FN_Light_Attributes(
    Color=&ColorFromCurve(ColorCurve=DataInterface() {propertyValues {
       AlphaCurve {keys []}
       BlueCurve {keys [<at 0.2351 4 RCIM_Cubic>,<at 0.9742 2 RCIM_Cubic>]}
       GreenCurve {keys [<at 0.2351 0.7033 RCIM_Cubic>,<at 0.9742 0.1262 RCIM_Cubic>]}
       RedCurve {keys [<at 0.2351 1.2129 RCIM_Cubic>,<at 0.9742 0.1711 RCIM_Cubic>]}
      }})
    Create Color Attribute=true
    Create Exponent Attribute=true
    Create Position Attribute=true
    Create Radius Attribute=true
    Enable Particle Light Enable=true
    Exponent=2
    Position=<0 0 50>
    Radius=&Multiply_Float(
     A=80
     B=&FloatFromCurve(FloatCurve=DataInterface() {propertyValues {Curve {keys [<at 0 1.8>,<at 0.2 2.05>,<at 0.5 2.2>,<at 1 1.8>]}}})
    )
    Scale Color=&FloatFromCurve(
     FloatCurve=DataInterface() {propertyValues {Curve {keys [
         {arriveTangent=5.626, interpMode=RCIM_Cubic, leaveTangent=5.626, tangentMode=RCTM_User, time=0, value=0.6}
         {arriveTangent=0.0895, interpMode=RCIM_Cubic, leaveTangent=0.0895, tangentMode=RCTM_User, time=0.1, value=1}
         {arriveTangent=-0.8586, interpMode=RCIM_Cubic, leaveTangent=-0.8586, tangentMode=RCTM_User, time=0.815, value=0.72}
         {arriveTangent=-4.4162, interpMode=RCIM_Cubic, leaveTangent=-4.4162, tangentMode=RCTM_User, time=1, value=0}
        ]}}}
     Scale Curve=&Multiply_Float(
      A=&FN_FrontEndFloatModifier(Gameplay=15, Front End=5)
      B=&FNTimeOfDayScalar(Night Scale=0.5, Morning Scale=1, Evening Scale=0.8)
     )
    )
   )
   /SetVariables(Particles.Dark_Color:LinearColor=&MakeLinearColorFromVectorAndFloat(
     "Vector (RGB)"=&MakeVectorFromLinearColorRGB(Color=<rgba 0.0043 0 0.02 1>)
     "Float (Alpha)"=&FloatFromCurve(
      FloatCurve=DataInterface() {propertyValues {Curve {keys [{arriveTangent=-1.4568, interpMode=RCIM_Cubic, leaveTangent=-1.4568, tangentMode=RCTM_User, time=0, value=1}, <at 1 0 RCIM_Cubic>]}}}
      Scale Curve=2
     )
    )) {isSetParameters=true}
   /SetVariables(Particles.Dark_Size:Vector2f=&Multiply_Vector2D(A=$Particles.SpriteSize, B=&Vector2DFromFloat(Value=2))) {isSetParameters=true}
   /SetVariables(Particles.Dark_Dyn:Vector4f=&MakeVector4(X=2, Y=1, Z=0, W=0)) {isSetParameters=true}
   /SetVariables(Particles.Light_Size:Vector2f=&Multiply_Vector2D(
     A=<300 200>
     B=&Vector2DFromFloat(Value=&FloatFromCurve(FloatCurve=DataInterface() {propertyValues {Curve {keys [
           {arriveTangent=4.0727, interpMode=RCIM_Cubic, leaveTangent=4.0727, tangentMode=RCTM_User, time=0, value=0}
           {arriveTangent=0.7027, interpMode=RCIM_Cubic, leaveTangent=0.7027, tangentMode=RCTM_User, time=0.2, value=0.7}
           <at 1 1 RCIM_Cubic>
          ]}}}))
    )) {isSetParameters=true}
   /SetVariables(Particles.Light-Color:LinearColor=&MakeLinearColorFromVectorAndFloat(
     "Vector (RGB)"=&MakeVectorFromLinearColorRGB(Color=&ColorFromCurve(ColorCurve=DataInterface() {propertyValues {
         AlphaCurve {keys []}
         BlueCurve {keys [<at 0.0805 5>, {arriveTangent=-157.524, leaveTangent=-157.524, time=0.4002, value=3}, <at 0.7494 1>]}
         GreenCurve {keys [<at 0.0805 0.5>, {arriveTangent=-10.1007, leaveTangent=-10.1007, time=0.4002, value=0.1924}, <at 0.7494 0.0641>]}
         RedCurve {keys [<at 0.0805 1.4884>, {arriveTangent=-42.4815, leaveTangent=-42.4815, time=0.4002, value=0.809}, <at 0.7494 0.2697>]}
        }}))
     "Float (Alpha)"=&FloatFromCurve(
      FloatCurve=DataInterface() {propertyValues {Curve {keys [<at 0 1>,<at 1 0>]}}}
      Scale Curve=&MakeCustomFloatFromBool(Bool=$User.bIsFrontEndPreview, True Float=0.2, False Float=1)
     )
    )) {isSetParameters=true}
  ]
  rendererClasses [&NiagaraSpriteRendererProperties, &NiagaraLightRendererProperties]
  renderers [
   {rendererClass=&NiagaraSpriteRendererProperties}
   {rendererClass=&NiagaraLightRendererProperties, rendererIndex=1}
   {rendererClass=&NiagaraSpriteRendererProperties, rendererIndex=2}
  ]
 }
 Glow001 {
  emitterUpdateScript [
   &EmitterState(
    Inactive Response=::NewEnumerator1
    Life Cycle Mode=::NewEnumerator1
    Loop Behavior=::NewEnumerator1
    Loop Delay=1.6
    Loop Duration=0.5
    UseLoopDelay=false
   )
   &SpawnBurst_Instantaneous(Use Spawn Probability=false, UseLoopCountLimit=false)
  ]
  particleSpawnScript [
   &InitializeParticle(
    Color=<rgba 0.5394 0.1282 2 1>
    Lifetime=0.25
    Sprite Size Mode=::NewEnumerator3
    Uniform Sprite Size=200
    UsePositionOffset=false
    Write Material Random=true
   )
   &CameraOffset(Camera Offset Amount=40)
  ]
  particleUpdateScript [
   &ParticleState()
   &ScaleSpriteSize(
    Uniform Curve Scale=&FN_FrontEndFloatModifier(Gameplay=1, Front End=0.8)
    Uniform Curve Sprite Scale=DataInterface() {propertyValues {Curve {keys [
        {arriveTangent=5.1246, interpMode=RCIM_Cubic, leaveTangent=5.1246, tangentMode=RCTM_User, time=0, value=0}
        {arriveTangent=1.7949, interpMode=RCIM_Cubic, leaveTangent=1.7949, tangentMode=RCTM_User, time=0.3, value=1.5}
        <at 1 2 RCIM_Cubic>
       ]}}}
   )
   &ScaleColor(
    Scale Alpha=&FloatFromCurve(
     FloatCurve=DataInterface() {propertyValues {Curve {keys [
         {arriveTangent=-2.7027, interpMode=RCIM_Cubic, leaveTangent=-2.7027, tangentMode=RCTM_User, time=0, value=1}
         {arriveTangent=-1, interpMode=RCIM_Cubic, leaveTangent=-1, time=0.312, value=0.3309}
         <at 1 0>
        ]}}}
     Scale Curve=&Multiply_Float(
      A=&MakeCustomFloatFromBool(Bool=$User.bIsFrontEndPreview, True Float=0.5, False Float=1)
      B=&FNTimeOfDayScalar(Night Scale=0.8, Morning Scale=1, Evening Scale=0.8)
     )
    )
    ScaleA=true
    ScaleRGB=true
   )
  ]
  rendererClasses [&NiagaraSpriteRendererProperties]
  renderers [{rendererClass=&NiagaraSpriteRendererProperties}]
 }
 Particle001 {
  emitterUpdateScript [
   &EmitterState(Inactive Response=::NewEnumerator1, Life Cycle Mode=::NewEnumerator1, Loop Behavior=::NewEnumerator1, Loop Duration=1, UseLoopDelay=false)
   &SpawnBurst_Instantaneous(Spawn Count=50, Use Spawn Probability=false, UseLoopCountLimit=false)
  ]
  particleSpawnScript [
   &PartitionParticles()
   &InitializeParticle(
    Color=<rgba 1.8703 0 10 1>
    Lifetime Max=0.8
    Lifetime Min=0.1
    Lifetime Mode=::NewEnumerator1
    Sprite Size Max=<2 4>
    Sprite Size Min=<1.5 2>
    Sprite Size Mode=::NewEnumerator2
    UsePositionOffset=true
    Write Material Random=true
   )
   &ShapeLocation(
    MaskSpawnGroup=false
    Offset Mode=::NewEnumerator0
    Override Seed=false
    Rotation Mode=::NewEnumerator1
    Sphere Radius=20
    UseOwnerScale=true
   )
   &AddVelocity(
    Cone Angle=110
    Cone Axis=&SelectVectorFromArray(
     Direct Array Index=$Particles.PartitionParticles.Partition
     Vector Selection Array=DataInterface() {propertyValues {FloatData [<1 0 0>, <-1 0 0>]}}
    )
    Module.UseConeFalloff=false
    Module.UseVelDistribution=false
    Override Seed=false
    Velocity Mode=::NewEnumerator2
    Velocity Speed=&RandomRangeFloat(Minimum=50, Maximum=400)
   )
  ]
  particleUpdateScript [
   &ParticleState()
   &ScaleSpriteSize(Uniform Curve Sprite Scale=DataInterface() {propertyValues {Curve {keys [
        <at 0 0>
        {arriveTangent=1.5, leaveTangent=1.5, time=0.0854, value=0.85}
        <at 0.2705 0.0312>
        <at 0.3267 0.9335>
        {arriveTangent=-0.25, leaveTangent=-0.25, time=0.4998, value=0.1033}
        {arriveTangent=-0.5, leaveTangent=-0.5, time=0.6227, value=0.9345}
        <at 0.8 0.05>
        <at 0.9 0.9>
        <at 1 0>
       ]}}})
   &CurlNoiseForce(Noise Strength=1200, Pan Noise=true, Pan Noise Field=<0 0.2 0.2>)
   &Drag(Drag=2, Use Linear Drag=true, Use Rotational Drag=true)
   &AccelerationForce(Acceleration=<0 0 200>)
   &SolveForcesAndVelocity(Clamp Velocity=false, Limit Acceleration=false)
   &ScaleColor(
    Scale Alpha=&FloatFromCurve(
     FloatCurve=DataInterface() {propertyValues {Curve {keys [
         {interpMode=RCIM_Cubic, tangentMode=RCTM_User, time=0, value=0}
         {interpMode=RCIM_Cubic, tangentMode=RCTM_User, time=0.5, value=1}
         {tangentMode=RCTM_User, time=1, value=0}
        ]}}}
     Scale Curve=&MakeCustomFloatFromBool(Bool=$User.bIsFrontEndPreview, True Float=0.5, False Float=1)
    )
    ScaleA=true
    ScaleRGB=true
   )
   &FN_Sprite_Attributes(
    Color=&MultiplyLinearColor(
     Linear Color=<rgba 0.2963 0.097 1 1>
     Scale Factor=&MakeVector4_FromFloat(Float=&FloatFromCurve(
       FloatCurve=DataInterface() {propertyValues {Curve {keys [
           {interpMode=RCIM_Cubic, tangentMode=RCTM_User, time=0, value=0}
           {interpMode=RCIM_Cubic, tangentMode=RCTM_User, time=0.5, value=0.8}
           {tangentMode=RCTM_User, time=1, value=0}
          ]}}}
       Scale Curve=&MakeCustomFloatFromBool(Bool=$User.bIsFrontEndPreview, True Float=0.5, False Float=1)
      ))
    )
    Create Color Attribute=true
    Create SpriteSize Attribute=true
    SpriteSize=&Multiply_Vector2D(
     A=&RandomRangeVector2D(Minimum=<3 5>, Maximum=<4 6>)
     B=&Vector2DFromFloat(Value=&FloatFromCurve(
       FloatCurve=DataInterface() {propertyValues {Curve {keys [<at 0 0>,<at 0.1 0.9>,<at 0.2 0.35>,<at 0.3 1>,<at 0.4 0.15>,<at 0.5 0.9>,<at 0.6 0.5>,<at 0.7 0.9>,<at 0.8 0.35>,<at 0.9 0.95>,<at 1 0>]}}}
       Scale Curve=&RandomRangeFloat(Minimum=3, Maximum=4)
      ))
    )
    Write to Module Namespace=true
   )
   &ScaleSpriteSizeBySpeed(Max Scale Factor=<1.5 10>)
  ]
  rendererClasses [&NiagaraSpriteRendererProperties]
  renderers [
   {rendererClass=&NiagaraSpriteRendererProperties}
   {rendererClass=&NiagaraSpriteRendererProperties, rendererIndex=1}
  ]
 }
 Shockwave02 {
  emitterUpdateScript [
   &EmitterState(Life Cycle Mode=::NewEnumerator1, Loop Behavior=::NewEnumerator1, Loop Duration=1, UseLoopDelay=true)
   &SpawnBurst_Instantaneous(Spawn Count=1, Use Spawn Probability=false)
   /SetVariables(Emitter.Ref=-4.5) {isSetParameters=true}
  ]
  particleSpawnScript [
   &InitializeParticle(
    Color Mode=::NewEnumerator0
    Lifetime=0.45
    Mesh Scale Mode=::NewEnumerator3
    Position Mode=::NewEnumerator1
    UsePositionOffset=false
    Write Material Random=true
   )
   &InitialMeshOrientation(Orientation Axis=<0 0 1>, Orientation Vector=<0 0 1>, Rotation=<0 0.25 0>, Use Orientation Vector=true, Use Rotation Vector=true)
   &InitialMeshOrientation(
    Orientation Axis=<0 0 0>
    Orientation Vector=<0 0 0>
    Rotation=&RandomRangeVector(Minimum=<0 0 0>, Maximum=<1 0 0>)
    Use Orientation Vector=false
    Use Rotation Vector=true
   ) {module=InitialMeshOrientation001}
   &DynamicMaterialParameters(
    Index 0 Param 1=&FloatFromCurve(FloatCurve=DataInterface() {propertyValues {Curve {keys [{arriveTangent=3.726, interpMode=RCIM_Cubic, leaveTangent=3.726, tangentMode=RCTM_User, time=0, value=-0.3}, <at 1 1>]}}})
    Index 0 Param 2=1
    Index 0 Param 3=1
    Module.Param0WriteEnabled=false
    Module.Param1WriteEnabled=true
    Module.Param2WriteEnabled=true
    Module.Param3WriteEnabled=false
   ) {module=DynamicMaterialParameters001}
   &CameraOffset(Camera Offset Amount=2)
  ]
  particleUpdateScript [
   &ParticleState()
   &ScaleColor(
    Linear Color Curve=DataInterface() {propertyValues {
      AlphaCurve {keys [<at 0 1>,<at 0.3576 1>,<at 1 0>]}
      BlueCurve {keys [<at 0.8 1>]}
      GreenCurve {keys [<at 0.8 0.0033>]}
      RedCurve {keys [<at 0.8 0.3685>]}
     }}
    Scale Mode=::NewEnumerator2
   )
   &DynamicMaterialParameters(
    Index 0 Param 1=&FloatFromCurve(FloatCurve=DataInterface() {propertyValues {Curve {keys [{arriveTangent=2.8297, interpMode=RCIM_Cubic, leaveTangent=2.8297, tangentMode=RCTM_User, time=0, value=-0.1}, <at 1 1>]}}})
    Index 0 Param 2=&FloatFromCurve(FloatCurve=DataInterface() {propertyValues {Curve {keys [{arriveTangent=-0.5818, interpMode=RCIM_Cubic, leaveTangent=-0.5818, tangentMode=RCTM_User, time=0, value=1}]}}})
    Index 0 Param 3=&FloatFromCurve(FloatCurve=DataInterface() {propertyValues {Curve {keys [<at 0.15 1>]}}})
    Module.Param0WriteEnabled=true
    Module.Param1WriteEnabled=false
    Module.Param2WriteEnabled=false
    Module.Param3WriteEnabled=false
   )
   &ScaleColor(
    Scale Alpha=&MakeCustomFloatFromBool(Bool=$User.bIsFrontEndPreview, True Float=0.3, False Float=0.5)
    Scale RGB=&VectorFromFloat(Value=&FloatFromCurve(FloatCurve=DataInterface() {propertyValues {Curve {keys [<at 0 25>,<at 0.2 10>,<at 1 0>]}}}))
    ScaleA=true
    ScaleRGB=true
   ) {module=ScaleColor001}
   &ScaleMeshSize(Scale Factor=&VectorFromFloat(Value=&FloatFromCurve(FloatCurve=DataInterface() {propertyValues {Curve {keys [
          {arriveTangent=2.5697, interpMode=RCIM_Cubic, leaveTangent=2.5697, tangentMode=RCTM_User, time=0, value=0.5}
          {arriveTangent=0.3018, interpMode=RCIM_Cubic, leaveTangent=0.3018, tangentMode=RCTM_User, time=0.2, value=1.05}
          {arriveTangent=0.078, interpMode=RCIM_Cubic, leaveTangent=0.078, tangentMode=RCTM_User, time=1, value=1.2}
         ]}}})))
  ]
  rendererClasses [&NiagaraMeshRendererProperties]
  renderers [{rendererClass=&NiagaraMeshRendererProperties, rendererIndex=0}]
 }
}
summary {
 systemName=NS_Emote_TidalNinja_Division
 emitters [
  {emitterName=Glow, rendererClasses [&NiagaraSpriteRendererProperties, &NiagaraLightRendererProperties]}
  {emitterName=Flow_Mesh_02, rendererClasses [&NiagaraMeshRendererProperties]}
  {emitterName=EnergyLeak001, rendererClasses [&NiagaraMeshRendererProperties]}
  {emitterName=Shockwave02, rendererClasses [&NiagaraMeshRendererProperties]}
  {emitterName=Particle001, rendererClasses [&NiagaraSpriteRendererProperties]}
  {emitterName=Glow001, rendererClasses [&NiagaraSpriteRendererProperties]}
 ]
 userVariables [
  User.bIsFrontEnd {type=NiagaraBool}
  User.bIsFrontEndPreview {type=NiagaraBool}
 ]
}
